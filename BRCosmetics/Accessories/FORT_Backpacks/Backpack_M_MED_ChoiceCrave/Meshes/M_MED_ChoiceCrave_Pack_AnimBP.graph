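# Post-pose AnimBP for the ChoiceCrave backpack cosmetic, adding springy secondary motion to its flowers and petals.
# The input pose passes through three additive offsets on the pack bone, then serial Anim Dynamics nodes, one per bone.
# Flower bodies simulate relative to the pack without gravity, while petals follow their flower base under scaled gravity.
# Root-relative simulation, angular springs, tight angular limits and scaled-down component motion restrain the sway.
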AnimBlueprint /BRCosmetics/Accessories/FORT_Backpacks/Backpack_M_MED_ChoiceCrave/Meshes/M_MED_ChoiceCrave_Pack_AnimBP
target_skeleton=&/M_MED_ChoiceCrave_Pack_Skeleton
graphs [
 AnimGraph {nodes [
   n1 = "Transform (Modify) Bone - Bone: pack"(ComponentPose=Local To Component(LocalPose=Input Pose()), Translation="0,0.25,0", Rotation="0,-4,-4") {
    rotationMode=BMM_Additive
    rotationSpace=BCS_BoneSpace
    translationMode=BMM_Additive
    translationSpace=BCS_BoneSpace
    boneToModify {boneName=pack}
    rotation {roll=-4, yaw=-4}
    translation {y=0.25}
   }
   n2 = "Transform (Modify) Bone - Bone: pack"(ComponentPose=@n1, Translation="0,-0.85,0", Rotation="0,0,0") {
    rotationMode=BMM_Additive
    rotationSpace=BCS_BoneSpace
    translationMode=BMM_Additive
    translationSpace=BCS_BoneSpace
    boneToModify {boneName=pack}
    translation {y=-0.85}
   }
   n3 = "Transform (Modify) Bone - Bone: pack"(ComponentPose=@n2, Translation="0,-1.3,0", Rotation="0,0,0") {
    rotationMode=BMM_Additive
    rotationSpace=BCS_BoneSpace
    translationMode=BMM_Additive
    translationSpace=BCS_BoneSpace
    boneToModify {boneName=pack}
    translation {y=-1.3}
   }
   n4 = Anim Dynamics - Bone: dyn_flower_01_cr!(
    ComponentPose=@n3
    Alpha=1
    GravityOverride:Vector="0,0,0"
    ComponentLinearAccScale:Vector="0.35,0.35,0.35"
    ComponentLinearVelScale:Vector="0.35,0.35,0.35"
    ExternalForce="0,0,0"
   ) -> (Pose) {
    angularDampingOverride=1
    angularSpringConstant=4000
    bAngularSpring=true
    bOverrideAngularDamping=true
    bUseGravityOverride=true
    gravityScale=0.35
    linearSpringConstant=10000
    simulationSpace=RootRelative
    worldSpaceMinimumScale=0
    boundBone {boneName=dyn_flower_01_cr}
    componentAppliedLinearAccClamp=<35000 35000 35000>
    componentLinearAccScale=<0.35 0.35 0.35>
    componentLinearVelScale=<0.35 0.35 0.35>
    physicsBodyDefinitions [{
      boundBone {boneName=dyn_flower_01_cr}
      boxExtents=<24 24 24>
      constraintSetup {angularLimitsMax=<0 10 10>, angularLimitsMin=<0 -10 -10>, angularTarget=<1 0 0>}
      localJointOffset=<2 0 0>
     }]
    relativeSpaceBone {boneName=pack}
    simSpaceSettings {maxAngularAcceleration=20, maxAngularVelocity=40, simSpaceAngularAlpha=0.5, externalAngularVelocity {y=1.5}}
   }
   n5 = Anim Dynamics - Bone: dyn_flower_02_lf!(
    ComponentPose=@n4
    Alpha=1
    GravityOverride:Vector="0,0,0"
    ComponentLinearAccScale:Vector="0.35,0.35,0.35"
    ComponentLinearVelScale:Vector="0.35,0.35,0.35"
    ExternalForce="0,0,0"
   ) -> (Pose) {
    angularDampingOverride=1
    angularSpringConstant=4000
    bAngularSpring=true
    bOverrideAngularDamping=true
    bUseGravityOverride=true
    gravityScale=0.35
    linearSpringConstant=10000
    simulationSpace=RootRelative
    worldSpaceMinimumScale=0
    boundBone {boneName=dyn_flower_02_lf}
    componentAppliedLinearAccClamp=<35000 35000 35000>
    componentLinearAccScale=<0.35 0.35 0.35>
    componentLinearVelScale=<0.35 0.35 0.35>
    physicsBodyDefinitions [{
      boundBone {boneName=dyn_flower_02_lf}
      boxExtents=<24 24 24>
      constraintSetup {angularLimitsMax=<0 10 20>, angularLimitsMin=<0 -8 -20>, angularTarget=<1 0 0>}
      localJointOffset=<2 0 0>
     }]
    relativeSpaceBone {boneName=pack}
    simSpaceSettings {maxAngularAcceleration=20, maxAngularVelocity=40, simSpaceAngularAlpha=0.5, externalAngularVelocity {y=1.5}}
   }
   n6 = Anim Dynamics - Bone: dyn_flower_03_lfbk!(
    ComponentPose=@n5
    Alpha=1
    GravityOverride:Vector="0,0,0"
    ComponentLinearAccScale:Vector="0.35,0.35,0.35"
    ComponentLinearVelScale:Vector="0.35,0.35,0.35"
    ExternalForce="0,0,0"
   ) -> (Pose) {
    angularDampingOverride=1
    angularSpringConstant=4000
    bAngularSpring=true
    bOverrideAngularDamping=true
    bUseGravityOverride=true
    gravityScale=0.35
    linearSpringConstant=10000
    simulationSpace=RootRelative
    worldSpaceMinimumScale=0
    boundBone {boneName=dyn_flower_03_lfbk}
    componentAppliedLinearAccClamp=<35000 35000 35000>
    componentLinearAccScale=<0.35 0.35 0.35>
    componentLinearVelScale=<0.35 0.35 0.35>
    physicsBodyDefinitions [{
      boundBone {boneName=dyn_flower_03_lfbk}
      boxExtents=<24 24 24>
      constraintSetup {angularLimitsMax=<0 5 5>, angularLimitsMin=<0 -5 -2.5>, angularTarget=<1 0 0>}
      localJointOffset=<2 0 0>
     }]
    relativeSpaceBone {boneName=pack}
    simSpaceSettings {maxAngularAcceleration=20, maxAngularVelocity=40, simSpaceAngularAlpha=0.5, externalAngularVelocity {y=1.5}}
   }
   n7 = Anim Dynamics - Bone: dyn_flower_04_rt!(
    ComponentPose=@n6
    Alpha=1
    GravityOverride:Vector="0,0,0"
    ComponentLinearAccScale:Vector="0.35,0.35,0.35"
    ComponentLinearVelScale:Vector="0.35,0.35,0.35"
    ExternalForce="0,0,0"
   ) -> (Pose) {
    angularDampingOverride=1
    angularSpringConstant=4000
    bAngularSpring=true
    bOverrideAngularDamping=true
    bUseGravityOverride=true
    gravityScale=0.35
    linearSpringConstant=10000
    simulationSpace=RootRelative
    worldSpaceMinimumScale=0
    boundBone {boneName=dyn_flower_04_rt}
    componentAppliedLinearAccClamp=<35000 35000 35000>
    componentLinearAccScale=<0.35 0.35 0.35>
    componentLinearVelScale=<0.35 0.35 0.35>
    physicsBodyDefinitions [{
      boundBone {boneName=dyn_flower_04_rt}
      boxExtents=<27 27 27>
      constraintSetup {angularLimitsMax=<0 10 1>, angularLimitsMin=<0 -5 -5>, angularTarget=<1 0 0>}
      localJointOffset=<2 0 0>
     }]
    relativeSpaceBone {boneName=pack}
    simSpaceSettings {maxAngularAcceleration=20, maxAngularVelocity=40, simSpaceAngularAlpha=0.5, externalAngularVelocity {y=1.5}}
   }
   n8 = Anim Dynamics - Bone: dyn_flower_05_bk!(
    ComponentPose=@n7
    Alpha=1
    GravityOverride:Vector="0,0,0"
    ComponentLinearAccScale:Vector="0.35,0.35,0.35"
    ComponentLinearVelScale:Vector="0.35,0.35,0.35"
    ExternalForce="0,0,0"
   ) -> (Pose) {
    angularDampingOverride=1
    angularSpringConstant=4000
    bAngularSpring=true
    bOverrideAngularDamping=true
    bUseGravityOverride=true
    gravityScale=0.35
    linearSpringConstant=10000
    simulationSpace=RootRelative
    worldSpaceMinimumScale=0
    boundBone {boneName=dyn_flower_05_bk}
    componentAppliedLinearAccClamp=<35000 35000 35000>
    componentLinearAccScale=<0.35 0.35 0.35>
    componentLinearVelScale=<0.35 0.35 0.35>
    physicsBodyDefinitions [{
      boundBone {boneName=dyn_flower_05_bk}
      boxExtents=<26 26 26>
      constraintSetup {angularLimitsMax=<0 5 5>, angularLimitsMin=<0 0 -5>, angularTarget=<1 0 0>}
      localJointOffset=<2 0 0>
     }]
    relativeSpaceBone {boneName=pack}
    simSpaceSettings {maxAngularAcceleration=20, maxAngularVelocity=40, simSpaceAngularAlpha=0.5, externalAngularVelocity {y=1.5}}
   }
   n9 = Anim Dynamics - Bone: dyn_petals01_01_cr!(
    ComponentPose=@n8
    Alpha=1
    GravityOverride:Vector="0,0,-980"
    ComponentLinearAccScale:Vector="0.35,0.35,0.35"
    ComponentLinearVelScale:Vector="0.35,0.35,0.35"
    ExternalForce="0,0,0"
   ) -> (Pose) {
    angularDampingOverride=0.9999
    angularSpringConstant=5000
    bAngularSpring=true
    bOverrideAngularDamping=true
    bUseGravityOverride=true
    gravityScale=0.35
    linearSpringConstant=10000
    simulationSpace=RootRelative
    worldSpaceMinimumScale=0
    boundBone {boneName=dyn_petals01_01_cr}
    componentAppliedLinearAccClamp=<35000 35000 35000>
    componentLinearAccScale=<0.35 0.35 0.35>
    componentLinearVelScale=<0.35 0.35 0.35>
    gravityOverride {z=-980}
    physicsBodyDefinitions [{
      boundBone {boneName=dyn_petals01_01_cr}
      boxExtents=<14 14 14>
      constraintSetup {angularLimitsMax=<0 10 10>, angularLimitsMin=<0 -10 -10>, angularTarget=<1 0 0>}
      localJointOffset=<3 0 0>
     }]
    relativeSpaceBone {boneName=base_flower_01_cr}
    simSpaceSettings {maxAngularAcceleration=35, maxAngularVelocity=10, simSpaceAngularAlpha=0.6, externalAngularVelocity {y=3}}
   }
   n10 = Anim Dynamics - Bone: dyn_petals01_02_cr!(
    ComponentPose=@n9
    Alpha=1
    GravityOverride:Vector="0,0,-980"
    ComponentLinearAccScale:Vector="0.35,0.35,0.35"
    ComponentLinearVelScale:Vector="0.35,0.35,0.35"
    ExternalForce="0,0,0"
   ) -> (Pose) {
    angularDampingOverride=0.9999
    angularSpringConstant=5000
    bAngularSpring=true
    bOverrideAngularDamping=true
    bUseGravityOverride=true
    gravityScale=0.35
    linearSpringConstant=10000
    simulationSpace=RootRelative
    worldSpaceMinimumScale=0
    boundBone {boneName=dyn_petals01_02_cr}
    componentAppliedLinearAccClamp=<35000 35000 35000>
    componentLinearAccScale=<0.35 0.35 0.35>
    componentLinearVelScale=<0.35 0.35 0.35>
    gravityOverride {z=-980}
    physicsBodyDefinitions [{
      boundBone {boneName=dyn_petals01_02_cr}
      boxExtents=<10 10 10>
      constraintSetup {angularLimitsMax=<0 15 15>, angularLimitsMin=<0 -15 -15>, angularTarget=<1 0 0>}
      localJointOffset=<3 0 0>
     }]
    relativeSpaceBone {boneName=base_flower_01_cr}
    simSpaceSettings {maxAngularAcceleration=35, maxAngularVelocity=10, simSpaceAngularAlpha=0.8, externalAngularVelocity {y=3}}
   }
   n11 = Anim Dynamics - Bone: dyn_petals02_01_cr!(
    ComponentPose=@n10
    Alpha=1
    GravityOverride:Vector="0,0,-980"
    ComponentLinearAccScale:Vector="0.35,0.35,0.35"
    ComponentLinearVelScale:Vector="0.35,0.35,0.35"
    ExternalForce="0,0,0"
   ) -> (Pose) {
    angularDampingOverride=0.9999
    angularSpringConstant=5000
    bAngularSpring=true
    bOverrideAngularDamping=true
    bUseGravityOverride=true
    gravityScale=0.35
    linearSpringConstant=10000
    simulationSpace=RootRelative
    worldSpaceMinimumScale=0
    boundBone {boneName=dyn_petals02_01_cr}
    componentAppliedLinearAccClamp=<35000 35000 35000>
    componentLinearAccScale=<0.35 0.35 0.35>
    componentLinearVelScale=<0.35 0.35 0.35>
    gravityOverride {z=-980}
    physicsBodyDefinitions [{
      boundBone {boneName=dyn_petals02_01_cr}
      boxExtents=<12 12 12>
      constraintSetup {angularLimitsMax=<0 4 4>, angularLimitsMin=<0 -4 -4>, angularTarget=<1 0 0>}
      localJointOffset=<3 0 0>
     }]
    relativeSpaceBone {boneName=base_flower_01_cr}
    simSpaceSettings {maxAngularAcceleration=35, maxAngularVelocity=10, simSpaceAngularAlpha=0.6, externalAngularVelocity {y=3}}
   }
   n12 = Anim Dynamics - Bone: dyn_petals02_03_cr!(
    ComponentPose=@n11
    Alpha=1
    GravityOverride:Vector="0,0,-980"
    ComponentLinearAccScale:Vector="0.35,0.35,0.35"
    ComponentLinearVelScale:Vector="0.35,0.35,0.35"
    ExternalForce="0,0,0"
   ) -> (Pose) {
    angularDampingOverride=0.9999
    angularSpringConstant=5000
    bAngularSpring=true
    bOverrideAngularDamping=true
    bUseGravityOverride=true
    gravityScale=0.35
    linearSpringConstant=10000
    simulationSpace=RootRelative
    worldSpaceMinimumScale=0
    boundBone {boneName=dyn_petals02_03_cr}
    componentAppliedLinearAccClamp=<35000 35000 35000>
    componentLinearAccScale=<0.35 0.35 0.35>
    componentLinearVelScale=<0.35 0.35 0.35>
    gravityOverride {z=-980}
    physicsBodyDefinitions [{
      boundBone {boneName=dyn_petals02_03_cr}
      boxExtents=<8 8 8>
      constraintSetup {angularLimitsMax=<0 8 8>, angularLimitsMin=<0 -8 -8>, angularTarget=<1 0 0>}
      localJointOffset=<3 0 0>
     }]
    relativeSpaceBone {boneName=base_flower_01_cr}
    simSpaceSettings {maxAngularAcceleration=35, maxAngularVelocity=10, simSpaceAngularAlpha=0.8, externalAngularVelocity {y=3}}
   }
   n13 = Anim Dynamics - Bone: dyn_petals03_01_cr!(
    ComponentPose=@n12
    Alpha=1
    GravityOverride:Vector="0,0,-980"
    ComponentLinearAccScale:Vector="0.35,0.35,0.35"
    ComponentLinearVelScale:Vector="0.35,0.35,0.35"
    ExternalForce="0,0,0"
   ) -> (Pose) {
    angularDampingOverride=0.9999
    angularSpringConstant=5000
    bAngularSpring=true
    bOverrideAngularDamping=true
    bUseGravityOverride=true
    gravityScale=0.35
    linearSpringConstant=10000
    simulationSpace=RootRelative
    worldSpaceMinimumScale=0
    boundBone {boneName=dyn_petals03_01_cr}
    componentAppliedLinearAccClamp=<35000 35000 35000>
    componentLinearAccScale=<0.35 0.35 0.35>
    componentLinearVelScale=<0.35 0.35 0.35>
    gravityOverride {z=-980}
    physicsBodyDefinitions [{
      boundBone {boneName=dyn_petals03_01_cr}
      boxExtents=<12 12 12>
      constraintSetup {angularLimitsMax=<0 5 5>, angularLimitsMin=<0 -5 -5>, angularTarget=<1 0 0>}
      localJointOffset=<3 0 0>
     }]
    relativeSpaceBone {boneName=base_flower_01_cr}
    simSpaceSettings {maxAngularAcceleration=35, maxAngularVelocity=10, simSpaceAngularAlpha=0.6, externalAngularVelocity {y=3}}
   }
   n14 = Anim Dynamics - Bone: dyn_petals03_02_cr!(
    ComponentPose=@n13
    Alpha=1
    GravityOverride:Vector="0,0,-980"
    ComponentLinearAccScale:Vector="0.35,0.35,0.35"
    ComponentLinearVelScale:Vector="0.35,0.35,0.35"
    ExternalForce="0,0,0"
   ) -> (Pose) {
    angularDampingOverride=0.9999
    angularSpringConstant=5000
    bAngularSpring=true
    bOverrideAngularDamping=true
    bUseGravityOverride=true
    gravityScale=0.35
    linearSpringConstant=10000
    simulationSpace=RootRelative
    worldSpaceMinimumScale=0
    boundBone {boneName=dyn_petals03_02_cr}
    componentAppliedLinearAccClamp=<35000 35000 35000>
    componentLinearAccScale=<0.35 0.35 0.35>
    componentLinearVelScale=<0.35 0.35 0.35>
    gravityOverride {z=-980}
    physicsBodyDefinitions [{
      boundBone {boneName=dyn_petals03_02_cr}
      boxExtents=<8 8 8>
      constraintSetup {angularLimitsMax=<0 8 8>, angularLimitsMin=<0 8 -8>, angularTarget=<1 0 0>}
      localJointOffset=<3 0 0>
     }]
    relativeSpaceBone {boneName=base_flower_01_cr}
    simSpaceSettings {maxAngularAcceleration=35, maxAngularVelocity=10, simSpaceAngularAlpha=0.8, externalAngularVelocity {y=3}}
   }
   n15 = Anim Dynamics - Bone: dyn_petals04_01_cr!(
    ComponentPose=@n14
    Alpha=1
    GravityOverride:Vector="0,0,-980"
    ComponentLinearAccScale:Vector="0.35,0.35,0.35"
    ComponentLinearVelScale:Vector="0.35,0.35,0.35"
    ExternalForce="0,0,0"
   ) -> (Pose) {
    angularDampingOverride=0.9999
    angularSpringConstant=5000
    bAngularSpring=true
    bOverrideAngularDamping=true
    bUseGravityOverride=true
    gravityScale=0.35
    linearSpringConstant=10000
    simulationSpace=RootRelative
    worldSpaceMinimumScale=0
    boundBone {boneName=dyn_petals04_01_cr}
    componentAppliedLinearAccClamp=<35000 35000 35000>
    componentLinearAccScale=<0.35 0.35 0.35>
    componentLinearVelScale=<0.35 0.35 0.35>
    gravityOverride {z=-980}
    physicsBodyDefinitions [{
      boundBone {boneName=dyn_petals04_01_cr}
      boxExtents=<13 13 13>
      constraintSetup {angularLimitsMax=<0 10 10>, angularLimitsMin=<0 -10 -10>, angularTarget=<1 0 0>}
      localJointOffset=<3 0 0>
     }]
    relativeSpaceBone {boneName=base_flower_01_cr}
    simSpaceSettings {maxAngularAcceleration=35, maxAngularVelocity=10, simSpaceAngularAlpha=0.6, externalAngularVelocity {y=3}}
   }
   n16 = Anim Dynamics - Bone: dyn_petals04_02_cr!(
    ComponentPose=@n15
    Alpha=1
    GravityOverride:Vector="0,0,-980"
    ComponentLinearAccScale:Vector="0.35,0.35,0.35"
    ComponentLinearVelScale:Vector="0.35,0.35,0.35"
    ExternalForce="0,0,0"
   ) -> (Pose) {
    angularDampingOverride=0.9999
    angularSpringConstant=5000
    bAngularSpring=true
    bOverrideAngularDamping=true
    bUseGravityOverride=true
    gravityScale=0.35
    linearSpringConstant=10000
    simulationSpace=RootRelative
    worldSpaceMinimumScale=0
    boundBone {boneName=dyn_petals04_02_cr}
    componentAppliedLinearAccClamp=<35000 35000 35000>
    componentLinearAccScale=<0.35 0.35 0.35>
    componentLinearVelScale=<0.35 0.35 0.35>
    gravityOverride {z=-980}
    physicsBodyDefinitions [{
      boundBone {boneName=dyn_petals04_02_cr}
      boxExtents=<17 17 17>
      constraintSetup {angularLimitsMax=<0 15 15>, angularLimitsMin=<0 -15 -15>, angularTarget=<1 0 0>}
      localJointOffset=<2 0 0>
     }]
    relativeSpaceBone {boneName=base_flower_01_cr}
    simSpaceSettings {maxAngularAcceleration=35, maxAngularVelocity=10, simSpaceAngularAlpha=0.8, externalAngularVelocity {y=3}}
   }
   n17 = Anim Dynamics - Bone: dyn_petals05_01_cr!(
    ComponentPose=@n16
    Alpha=1
    GravityOverride:Vector="0,0,-980"
    ComponentLinearAccScale:Vector="0.35,0.35,0.35"
    ComponentLinearVelScale:Vector="0.35,0.35,0.35"
    ExternalForce="0,0,0"
   ) -> (Pose) {
    angularDampingOverride=0.9999
    angularSpringConstant=5000
    bAngularSpring=true
    bOverrideAngularDamping=true
    bUseGravityOverride=true
    gravityScale=0.35
    linearSpringConstant=10000
    simulationSpace=RootRelative
    worldSpaceMinimumScale=0
    boundBone {boneName=dyn_petals05_01_cr}
    componentAppliedLinearAccClamp=<35000 35000 35000>
    componentLinearAccScale=<0.35 0.35 0.35>
    componentLinearVelScale=<0.35 0.35 0.35>
    gravityOverride {z=-980}
    physicsBodyDefinitions [{
      boundBone {boneName=dyn_petals05_01_cr}
      boxExtents=<13 13 13>
      constraintSetup {angularLimitsMax=<0 10 10>, angularLimitsMin=<0 -10 -10>, angularTarget=<1 0 0>}
      localJointOffset=<3 0 0>
     }]
    relativeSpaceBone {boneName=base_flower_01_cr}
    simSpaceSettings {maxAngularAcceleration=35, maxAngularVelocity=10, simSpaceAngularAlpha=0.6, externalAngularVelocity {y=3}}
   }
   n18 = Anim Dynamics - Bone: dyn_petals05_02_cr!(
    ComponentPose=@n17
    Alpha=1
    GravityOverride:Vector="0,0,-980"
    ComponentLinearAccScale:Vector="0.35,0.35,0.35"
    ComponentLinearVelScale:Vector="0.35,0.35,0.35"
    ExternalForce="0,0,0"
   ) -> (Pose) {
    angularDampingOverride=0.9999
    angularSpringConstant=5000
    bAngularSpring=true
    bOverrideAngularDamping=true
    bUseGravityOverride=true
    gravityScale=0.35
    linearSpringConstant=10000
    simulationSpace=RootRelative
    worldSpaceMinimumScale=0
    boundBone {boneName=dyn_petals05_02_cr}
    componentAppliedLinearAccClamp=<35000 35000 35000>
    componentLinearAccScale=<0.35 0.35 0.35>
    componentLinearVelScale=<0.35 0.35 0.35>
    gravityOverride {z=-980}
    physicsBodyDefinitions [{
      boundBone {boneName=dyn_petals05_02_cr}
      boxExtents=<17 17 17>
      constraintSetup {angularLimitsMax=<0 15 15>, angularLimitsMin=<0 -15 -15>, angularTarget=<1 0 0>}
      localJointOffset=<2.5 0 0>
     }]
    relativeSpaceBone {boneName=base_flower_01_cr}
    simSpaceSettings {maxAngularAcceleration=35, maxAngularVelocity=10, simSpaceAngularAlpha=0.8, externalAngularVelocity {y=3}}
   }
   n19 = Anim Dynamics - Bone: dyn_petals06_01_cr!(
    ComponentPose=@n18
    Alpha=1
    GravityOverride:Vector="0,0,-980"
    ComponentLinearAccScale:Vector="0.35,0.35,0.35"
    ComponentLinearVelScale:Vector="0.35,0.35,0.35"
    ExternalForce="0,0,0"
   ) -> (Pose) {
    angularDampingOverride=0.9999
    angularSpringConstant=5000
    bAngularSpring=true
    bOverrideAngularDamping=true
    bUseGravityOverride=true
    gravityScale=0.35
    linearSpringConstant=10000
    simulationSpace=RootRelative
    worldSpaceMinimumScale=0
    boundBone {boneName=dyn_petals06_01_cr}
    componentAppliedLinearAccClamp=<35000 35000 35000>
    componentLinearAccScale=<0.35 0.35 0.35>
    componentLinearVelScale=<0.35 0.35 0.35>
    gravityOverride {z=-980}
    physicsBodyDefinitions [{
      boundBone {boneName=dyn_petals06_01_cr}
      boxExtents=<14 14 14>
      constraintSetup {angularLimitsMax=<0 10 10>, angularLimitsMin=<0 -10 -10>, angularTarget=<1 0 0>}
      localJointOffset=<3 0 0>
     }]
    relativeSpaceBone {boneName=base_flower_01_cr}
    simSpaceSettings {maxAngularAcceleration=35, maxAngularVelocity=10, simSpaceAngularAlpha=0.6, externalAngularVelocity {y=3}}
   }
   n20 = Anim Dynamics - Bone: dyn_petals06_02_cr!(
    ComponentPose=@n19
    Alpha=1
    GravityOverride:Vector="0,0,-980"
    ComponentLinearAccScale:Vector="0.35,0.35,0.35"
    ComponentLinearVelScale:Vector="0.35,0.35,0.35"
    ExternalForce="0,0,0"
   ) -> (Pose) {
    angularDampingOverride=0.9999
    angularSpringConstant=5000
    bAngularSpring=true
    bOverrideAngularDamping=true
    bUseGravityOverride=true
    gravityScale=0.35
    linearSpringConstant=10000
    simulationSpace=RootRelative
    worldSpaceMinimumScale=0
    boundBone {boneName=dyn_petals06_02_cr}
    componentAppliedLinearAccClamp=<35000 35000 35000>
    componentLinearAccScale=<0.35 0.35 0.35>
    componentLinearVelScale=<0.35 0.35 0.35>
    gravityOverride {z=-980}
    physicsBodyDefinitions [{
      boundBone {boneName=dyn_petals06_02_cr}
      boxExtents=<10 10 10>
      constraintSetup {angularLimitsMax=<0 15 15>, angularLimitsMin=<0 -15 -15>, angularTarget=<1 0 0>}
      localJointOffset=<3 0 0>
     }]
    relativeSpaceBone {boneName=base_flower_01_cr}
    simSpaceSettings {maxAngularAcceleration=35, maxAngularVelocity=10, simSpaceAngularAlpha=0.8, externalAngularVelocity {y=3}}
   }
   n21 = Anim Dynamics - Bone: dyn_petals01_01_lf!(
    ComponentPose=@n20
    Alpha=1
    GravityOverride:Vector="0,0,-980"
    ComponentLinearAccScale:Vector="0.35,0.35,0.35"
    ComponentLinearVelScale:Vector="0.35,0.35,0.35"
    ExternalForce="0,0,0"
   ) -> (Pose) {
    angularDampingOverride=0.9999
    angularSpringConstant=5000
    bAngularSpring=true
    bOverrideAngularDamping=true
    bUseGravityOverride=true
    gravityScale=0.35
    linearSpringConstant=10000
    simulationSpace=RootRelative
    worldSpaceMinimumScale=0
    boundBone {boneName=dyn_petals01_01_lf}
    componentAppliedLinearAccClamp=<35000 35000 35000>
    componentLinearAccScale=<0.35 0.35 0.35>
    componentLinearVelScale=<0.35 0.35 0.35>
    gravityOverride {z=-980}
    physicsBodyDefinitions [{
      boundBone {boneName=dyn_petals01_01_lf}
      boxExtents=<20 20 20>
      constraintSetup {angularLimitsMax=<0 10 10>, angularLimitsMin=<0 -10 -10>, angularTarget=<1 0 0>}
      localJointOffset=<3 0 0>
     }]
    relativeSpaceBone {boneName=base_flower_02_lf}
    simSpaceSettings {maxAngularAcceleration=35, maxAngularVelocity=10, simSpaceAngularAlpha=0.8, externalAngularVelocity {y=3}}
   }
   n22 = Anim Dynamics - Bone: dyn_petals01_02_lf!(
    ComponentPose=@n21
    Alpha=1
    GravityOverride:Vector="0,0,-980"
    ComponentLinearAccScale:Vector="0.35,0.35,0.35"
    ComponentLinearVelScale:Vector="0.35,0.35,0.35"
    ExternalForce="0,0,0"
   ) -> (Pose) {
    angularDampingOverride=0.9999
    angularSpringConstant=5000
    bAngularSpring=true
    bOverrideAngularDamping=true
    bUseGravityOverride=true
    gravityScale=0.35
    linearSpringConstant=10000
    simulationSpace=RootRelative
    worldSpaceMinimumScale=0
    boundBone {boneName=dyn_petals01_02_lf}
    componentAppliedLinearAccClamp=<35000 35000 35000>
    componentLinearAccScale=<0.35 0.35 0.35>
    componentLinearVelScale=<0.35 0.35 0.35>
    gravityOverride {z=-980}
    physicsBodyDefinitions [{
      boundBone {boneName=dyn_petals01_02_lf}
      boxExtents=<20 20 20>
      constraintSetup {angularLimitsMax=<0 15 15>, angularLimitsMin=<0 -15 -15>, angularTarget=<1 0 0>}
      localJointOffset=<3 0 0>
     }]
    relativeSpaceBone {boneName=base_flower_02_lf}
    simSpaceSettings {maxAngularAcceleration=35, maxAngularVelocity=10, simSpaceAngularAlpha=0.8, externalAngularVelocity {y=3}}
   }
   n23 = Anim Dynamics - Bone: dyn_petals02_01_lf!(
    ComponentPose=@n22
    Alpha=1
    GravityOverride:Vector="0,0,-980"
    ComponentLinearAccScale:Vector="0.35,0.35,0.35"
    ComponentLinearVelScale:Vector="0.35,0.35,0.35"
    ExternalForce="0,0,0"
   ) -> (Pose) {
    angularDampingOverride=0.9999
    angularSpringConstant=5000
    bAngularSpring=true
    bOverrideAngularDamping=true
    bUseGravityOverride=true
    gravityScale=0.35
    linearSpringConstant=10000
    simulationSpace=RootRelative
    worldSpaceMinimumScale=0
    boundBone {boneName=dyn_petals02_01_lf}
    componentAppliedLinearAccClamp=<35000 35000 35000>
    componentLinearAccScale=<0.35 0.35 0.35>
    componentLinearVelScale=<0.35 0.35 0.35>
    gravityOverride {z=-980}
    physicsBodyDefinitions [{
      boundBone {boneName=dyn_petals02_01_lf}
      boxExtents=<13 13 13>
      constraintSetup {angularLimitsMax=<0 8 8>, angularLimitsMin=<0 -8 -8>, angularTarget=<1 0 0>}
      localJointOffset=<3 0 0>
     }]
    relativeSpaceBone {boneName=base_flower_02_lf}
    simSpaceSettings {maxAngularAcceleration=35, maxAngularVelocity=10, simSpaceAngularAlpha=0.8, externalAngularVelocity {y=3}}
   }
   n24 = Anim Dynamics - Bone: dyn_petals02_02_lf!(
    ComponentPose=@n23
    Alpha=1
    GravityOverride:Vector="0,0,-980"
    ComponentLinearAccScale:Vector="0.35,0.35,0.35"
    ComponentLinearVelScale:Vector="0.35,0.35,0.35"
    ExternalForce="0,0,0"
   ) -> (Pose) {
    angularDampingOverride=0.9999
    angularSpringConstant=5000
    bAngularSpring=true
    bOverrideAngularDamping=true
    bUseGravityOverride=true
    gravityScale=0.35
    linearSpringConstant=10000
    simulationSpace=RootRelative
    worldSpaceMinimumScale=0
    boundBone {boneName=dyn_petals02_02_lf}
    componentAppliedLinearAccClamp=<35000 35000 35000>
    componentLinearAccScale=<0.35 0.35 0.35>
    componentLinearVelScale=<0.35 0.35 0.35>
    gravityOverride {z=-980}
    physicsBodyDefinitions [{
      boundBone {boneName=dyn_petals02_02_lf}
      boxExtents=<9 9 9>
      constraintSetup {angularLimitsMax=<0 10 10>, angularLimitsMin=<0 -10 -10>, angularTarget=<1 0 0>}
      localJointOffset=<3 0 0>
     }]
    relativeSpaceBone {boneName=base_flower_02_lf}
    simSpaceSettings {maxAngularAcceleration=35, maxAngularVelocity=10, simSpaceAngularAlpha=0.8, externalAngularVelocity {y=3}}
   }
   n25 = Anim Dynamics - Bone: dyn_petals03_01_lf!(
    ComponentPose=@n24
    Alpha=1
    GravityOverride:Vector="0,0,-980"
    ComponentLinearAccScale:Vector="0.35,0.35,0.35"
    ComponentLinearVelScale:Vector="0.35,0.35,0.35"
    ExternalForce="0,0,0"
   ) -> (Pose) {
    angularDampingOverride=0.9999
    angularSpringConstant=5000
    bAngularSpring=true
    bOverrideAngularDamping=true
    bUseGravityOverride=true
    gravityScale=0.35
    linearSpringConstant=10000
    simulationSpace=RootRelative
    worldSpaceMinimumScale=0
    boundBone {boneName=dyn_petals03_01_lf}
    componentAppliedLinearAccClamp=<35000 35000 35000>
    componentLinearAccScale=<0.35 0.35 0.35>
    componentLinearVelScale=<0.35 0.35 0.35>
    gravityOverride {z=-980}
    physicsBodyDefinitions [{
      boundBone {boneName=dyn_petals03_01_lf}
      boxExtents=<13 13 13>
      constraintSetup {angularLimitsMax=<0 10 10>, angularLimitsMin=<0 -10 -10>, angularTarget=<1 0 0>}
      localJointOffset=<3 0 0>
     }]
    relativeSpaceBone {boneName=base_flower_02_lf}
    simSpaceSettings {maxAngularAcceleration=35, maxAngularVelocity=10, simSpaceAngularAlpha=0.8, externalAngularVelocity {y=3}}
   }
   n26 = Anim Dynamics - Bone: dyn_petals03_02_lf!(
    ComponentPose=@n25
    Alpha=1
    GravityOverride:Vector="0,0,-980"
    ComponentLinearAccScale:Vector="0.35,0.35,0.35"
    ComponentLinearVelScale:Vector="0.35,0.35,0.35"
    ExternalForce="0,0,0"
   ) -> (Pose) {
    angularDampingOverride=0.9999
    angularSpringConstant=5000
    bAngularSpring=true
    bOverrideAngularDamping=true
    bUseGravityOverride=true
    gravityScale=0.35
    linearSpringConstant=10000
    simulationSpace=RootRelative
    worldSpaceMinimumScale=0
    boundBone {boneName=dyn_petals03_02_lf}
    componentAppliedLinearAccClamp=<35000 35000 35000>
    componentLinearAccScale=<0.35 0.35 0.35>
    componentLinearVelScale=<0.35 0.35 0.35>
    gravityOverride {z=-980}
    physicsBodyDefinitions [{
      boundBone {boneName=dyn_petals03_02_lf}
      boxExtents=<9 9 9>
      constraintSetup {angularLimitsMax=<0 15 15>, angularLimitsMin=<0 -15 -15>, angularTarget=<1 0 0>}
      localJointOffset=<3 0 0>
     }]
    relativeSpaceBone {boneName=base_flower_02_lf}
    simSpaceSettings {maxAngularAcceleration=35, maxAngularVelocity=10, simSpaceAngularAlpha=0.8, externalAngularVelocity {y=3}}
   }
   n27 = Anim Dynamics - Bone: dyn_petals04_01_lf!(
    ComponentPose=@n26
    Alpha=1
    GravityOverride:Vector="0,0,-980"
    ComponentLinearAccScale:Vector="0.35,0.35,0.35"
    ComponentLinearVelScale:Vector="0.35,0.35,0.35"
    ExternalForce="0,0,0"
   ) -> (Pose) {
    angularDampingOverride=0.9999
    angularSpringConstant=5000
    bAngularSpring=true
    bOverrideAngularDamping=true
    bUseGravityOverride=true
    gravityScale=0.35
    linearSpringConstant=10000
    simulationSpace=RootRelative
    worldSpaceMinimumScale=0
    boundBone {boneName=dyn_petals04_01_lf}
    componentAppliedLinearAccClamp=<35000 35000 35000>
    componentLinearAccScale=<0.35 0.35 0.35>
    componentLinearVelScale=<0.35 0.35 0.35>
    gravityOverride {z=-980}
    physicsBodyDefinitions [{
      boundBone {boneName=dyn_petals04_01_lf}
      boxExtents=<13 13 13>
      constraintSetup {angularLimitsMax=<0 10 10>, angularLimitsMin=<0 -10 -10>, angularTarget=<1 0 0>}
      localJointOffset=<3 0 0>
     }]
    relativeSpaceBone {boneName=base_flower_02_lf}
    simSpaceSettings {maxAngularAcceleration=35, maxAngularVelocity=10, simSpaceAngularAlpha=0.8, externalAngularVelocity {y=3}}
   }
   n28 = Anim Dynamics - Bone: dyn_petals04_02_lf!(
    ComponentPose=@n27
    Alpha=1
    GravityOverride:Vector="0,0,-980"
    ComponentLinearAccScale:Vector="0.35,0.35,0.35"
    ComponentLinearVelScale:Vector="0.35,0.35,0.35"
    ExternalForce="0,0,0"
   ) -> (Pose) {
    angularDampingOverride=0.9999
    angularSpringConstant=5000
    bAngularSpring=true
    bOverrideAngularDamping=true
    bUseGravityOverride=true
    gravityScale=0.35
    linearSpringConstant=10000
    simulationSpace=RootRelative
    worldSpaceMinimumScale=0
    boundBone {boneName=dyn_petals04_02_lf}
    componentAppliedLinearAccClamp=<35000 35000 35000>
    componentLinearAccScale=<0.35 0.35 0.35>
    componentLinearVelScale=<0.35 0.35 0.35>
    gravityOverride {z=-980}
    physicsBodyDefinitions [{
      boundBone {boneName=dyn_petals04_02_lf}
      boxExtents=<7 7 7>
      constraintSetup {angularLimitsMax=<0 15 15>, angularLimitsMin=<0 -15 -15>, angularTarget=<1 0 0>}
      localJointOffset=<3 0 0>
     }]
    relativeSpaceBone {boneName=base_flower_02_lf}
    simSpaceSettings {maxAngularAcceleration=35, maxAngularVelocity=10, simSpaceAngularAlpha=0.8, externalAngularVelocity {y=3}}
   }
   n29 = Anim Dynamics - Bone: dyn_petals05_01_lf!(
    ComponentPose=@n28
    Alpha=1
    GravityOverride:Vector="0,0,-980"
    ComponentLinearAccScale:Vector="0.35,0.35,0.35"
    ComponentLinearVelScale:Vector="0.35,0.35,0.35"
    ExternalForce="0,0,0"
   ) -> (Pose) {
    angularDampingOverride=0.9999
    angularSpringConstant=5000
    bAngularSpring=true
    bOverrideAngularDamping=true
    bUseGravityOverride=true
    gravityScale=0.35
    linearSpringConstant=10000
    simulationSpace=RootRelative
    worldSpaceMinimumScale=0
    boundBone {boneName=dyn_petals05_01_lf}
    componentAppliedLinearAccClamp=<35000 35000 35000>
    componentLinearAccScale=<0.35 0.35 0.35>
    componentLinearVelScale=<0.35 0.35 0.35>
    gravityOverride {z=-980}
    physicsBodyDefinitions [{
      boundBone {boneName=dyn_petals05_01_lf}
      boxExtents=<18 18 18>
      constraintSetup {angularLimitsMax=<0 10 10>, angularLimitsMin=<0 -10 -10>, angularTarget=<1 0 0>}
      localJointOffset=<3 0 0>
     }]
    relativeSpaceBone {boneName=base_flower_02_lf}
    simSpaceSettings {maxAngularAcceleration=35, maxAngularVelocity=10, simSpaceAngularAlpha=0.8, externalAngularVelocity {y=3}}
   }
   n30 = Anim Dynamics - Bone: dyn_petals05_02_lf!(
    ComponentPose=@n29
    Alpha=1
    GravityOverride:Vector="0,0,-980"
    ComponentLinearAccScale:Vector="0.35,0.35,0.35"
    ComponentLinearVelScale:Vector="0.35,0.35,0.35"
    ExternalForce="0,0,0"
   ) -> (Pose) {
    angularDampingOverride=0.9999
    angularSpringConstant=5000
    bAngularSpring=true
    bOverrideAngularDamping=true
    bUseGravityOverride=true
    gravityScale=0.35
    linearSpringConstant=10000
    simulationSpace=RootRelative
    worldSpaceMinimumScale=0
    boundBone {boneName=dyn_petals05_02_lf}
    componentAppliedLinearAccClamp=<35000 35000 35000>
    componentLinearAccScale=<0.35 0.35 0.35>
    componentLinearVelScale=<0.35 0.35 0.35>
    gravityOverride {z=-980}
    physicsBodyDefinitions [{
      boundBone {boneName=dyn_petals05_02_lf}
      boxExtents=<15 15 15>
      constraintSetup {angularLimitsMax=<0 15 15>, angularLimitsMin=<0 -15 -15>, angularTarget=<1 0 0>}
      localJointOffset=<3 0 0>
     }]
    relativeSpaceBone {boneName=base_flower_02_lf}
    simSpaceSettings {maxAngularAcceleration=35, maxAngularVelocity=10, simSpaceAngularAlpha=0.8, externalAngularVelocity {y=3}}
   }
   n31 = Anim Dynamics - Bone: dyn_petals01_01_lfbk!(
    ComponentPose=@n30
    Alpha=1
    GravityOverride:Vector="0,0,-980"
    ComponentLinearAccScale:Vector="0.35,0.35,0.35"
    ComponentLinearVelScale:Vector="0.35,0.35,0.35"
    ExternalForce="0,0,0"
   ) -> (Pose) {
    angularDampingOverride=0.9999
    angularSpringConstant=5000
    bAngularSpring=true
    bOverrideAngularDamping=true
    bUseGravityOverride=true
    gravityScale=0.35
    linearSpringConstant=10000
    simulationSpace=RootRelative
    worldSpaceMinimumScale=0
    boundBone {boneName=dyn_petals01_01_lfbk}
    componentAppliedLinearAccClamp=<35000 35000 35000>
    componentLinearAccScale=<0.35 0.35 0.35>
    componentLinearVelScale=<0.35 0.35 0.35>
    gravityOverride {z=-980}
    physicsBodyDefinitions [{
      boundBone {boneName=dyn_petals01_01_lfbk}
      boxExtents=<16 16 16>
      constraintSetup {angularLimitsMax=<0 10 10>, angularLimitsMin=<0 -10 -10>, angularTarget=<1 0 0>}
      localJointOffset=<3 0 0>
     }]
    relativeSpaceBone {boneName=base_flower_03_lfbk}
    simSpaceSettings {maxAngularAcceleration=35, maxAngularVelocity=10, simSpaceAngularAlpha=0.8, externalAngularVelocity {y=3}}
   }
   n32 = Anim Dynamics - Bone: dyn_petals01_02_lfbk!(
    ComponentPose=@n31
    Alpha=1
    GravityOverride:Vector="0,0,-980"
    ComponentLinearAccScale:Vector="0.35,0.35,0.35"
    ComponentLinearVelScale:Vector="0.35,0.35,0.35"
    ExternalForce="0,0,0"
   ) -> (Pose) {
    angularDampingOverride=0.9999
    angularSpringConstant=5000
    bAngularSpring=true
    bOverrideAngularDamping=true
    bUseGravityOverride=true
    gravityScale=0.35
    linearSpringConstant=10000
    simulationSpace=RootRelative
    worldSpaceMinimumScale=0
    boundBone {boneName=dyn_petals01_02_lfbk}
    componentAppliedLinearAccClamp=<35000 35000 35000>
    componentLinearAccScale=<0.35 0.35 0.35>
    componentLinearVelScale=<0.35 0.35 0.35>
    gravityOverride {z=-980}
    physicsBodyDefinitions [{
      boundBone {boneName=dyn_petals01_02_lfbk}
      boxExtents=<14 14 14>
      constraintSetup {angularLimitsMax=<0 10 10>, angularLimitsMin=<0 -10 -10>, angularTarget=<1 0 0>}
      localJointOffset=<3 0 0>
     }]
    relativeSpaceBone {boneName=base_flower_03_lfbk}
    simSpaceSettings {maxAngularAcceleration=35, maxAngularVelocity=10, simSpaceAngularAlpha=0.8, externalAngularVelocity {y=3}}
   }
   n33 = Anim Dynamics - Bone: dyn_petals02_01_lfbk!(
    ComponentPose=@n32
    Alpha=1
    GravityOverride:Vector="0,0,-980"
    ComponentLinearAccScale:Vector="0.35,0.35,0.35"
    ComponentLinearVelScale:Vector="0.35,0.35,0.35"
    ExternalForce="0,0,0"
   ) -> (Pose) {
    angularDampingOverride=0.9999
    angularSpringConstant=5000
    bAngularSpring=true
    bOverrideAngularDamping=true
    bUseGravityOverride=true
    gravityScale=0.35
    linearSpringConstant=10000
    simulationSpace=RootRelative
    worldSpaceMinimumScale=0
    boundBone {boneName=dyn_petals02_01_lfbk}
    componentAppliedLinearAccClamp=<35000 35000 35000>
    componentLinearAccScale=<0.35 0.35 0.35>
    componentLinearVelScale=<0.35 0.35 0.35>
    gravityOverride {z=-980}
    physicsBodyDefinitions [{
      boundBone {boneName=dyn_petals02_01_lfbk}
      boxExtents=<13 13 13>
      constraintSetup {angularLimitsMax=<0 10 10>, angularLimitsMin=<0 -10 -10>, angularTarget=<1 0 0>}
      localJointOffset=<3 0 0>
     }]
    relativeSpaceBone {boneName=base_flower_03_lfbk}
    simSpaceSettings {maxAngularAcceleration=35, maxAngularVelocity=10, simSpaceAngularAlpha=0.8, externalAngularVelocity {y=3}}
   }
   n34 = Anim Dynamics - Bone: dyn_petals02_02_lfbk!(
    ComponentPose=@n33
    Alpha=1
    GravityOverride:Vector="0,0,-980"
    ComponentLinearAccScale:Vector="0.35,0.35,0.35"
    ComponentLinearVelScale:Vector="0.35,0.35,0.35"
    ExternalForce="0,0,0"
   ) -> (Pose) {
    angularDampingOverride=0.9999
    angularSpringConstant=5000
    bAngularSpring=true
    bOverrideAngularDamping=true
    bUseGravityOverride=true
    gravityScale=0.35
    linearSpringConstant=10000
    simulationSpace=RootRelative
    worldSpaceMinimumScale=0
    boundBone {boneName=dyn_petals02_02_lfbk}
    componentAppliedLinearAccClamp=<35000 35000 35000>
    componentLinearAccScale=<0.35 0.35 0.35>
    componentLinearVelScale=<0.35 0.35 0.35>
    gravityOverride {z=-980}
    physicsBodyDefinitions [{
      boundBone {boneName=dyn_petals02_02_lfbk}
      boxExtents=<11 11 11>
      constraintSetup {angularLimitsMax=<0 10 10>, angularLimitsMin=<0 -10 10>, angularTarget=<1 0 0>}
      localJointOffset=<3 0 0>
     }]
    relativeSpaceBone {boneName=base_flower_03_lfbk}
    simSpaceSettings {maxAngularAcceleration=35, maxAngularVelocity=10, simSpaceAngularAlpha=0.8, externalAngularVelocity {y=3}}
   }
   n35 = Anim Dynamics - Bone: dyn_petals03_01_lfbk!(
    ComponentPose=@n34
    Alpha=1
    GravityOverride:Vector="0,0,-980"
    ComponentLinearAccScale:Vector="0.35,0.35,0.35"
    ComponentLinearVelScale:Vector="0.35,0.35,0.35"
    ExternalForce="0,0,0"
   ) -> (Pose) {
    angularDampingOverride=0.9999
    angularSpringConstant=5000
    bAngularSpring=true
    bOverrideAngularDamping=true
    bUseGravityOverride=true
    gravityScale=0.35
    linearSpringConstant=10000
    simulationSpace=RootRelative
    worldSpaceMinimumScale=0
    boundBone {boneName=dyn_petals03_01_lfbk}
    componentAppliedLinearAccClamp=<35000 35000 35000>
    componentLinearAccScale=<0.35 0.35 0.35>
    componentLinearVelScale=<0.35 0.35 0.35>
    gravityOverride {z=-980}
    physicsBodyDefinitions [{
      boundBone {boneName=dyn_petals03_01_lfbk}
      boxExtents=<16 16 16>
      constraintSetup {angularLimitsMax=<0 30 15>, angularLimitsMin=<0 -40 -15>, angularTarget=<1 0 0>}
      localJointOffset=<3 0 0>
     }]
    relativeSpaceBone {boneName=base_flower_03_lfbk}
    simSpaceSettings {maxAngularAcceleration=35, maxAngularVelocity=10, simSpaceAngularAlpha=0.8, externalAngularVelocity {y=3}}
   }
   n36 = Anim Dynamics - Bone: dyn_petals04_01_lfbk!(
    ComponentPose=@n35
    Alpha=1
    GravityOverride:Vector="0,0,-980"
    ComponentLinearAccScale:Vector="0.35,0.35,0.35"
    ComponentLinearVelScale:Vector="0.35,0.35,0.35"
    ExternalForce="0,0,0"
   ) -> (Pose) {
    angularDampingOverride=0.9999
    angularSpringConstant=5000
    bAngularSpring=true
    bOverrideAngularDamping=true
    bUseGravityOverride=true
    gravityScale=0.35
    linearSpringConstant=10000
    simulationSpace=RootRelative
    worldSpaceMinimumScale=0
    boundBone {boneName=dyn_petals04_01_lfbk}
    componentAppliedLinearAccClamp=<35000 35000 35000>
    componentLinearAccScale=<0.35 0.35 0.35>
    componentLinearVelScale=<0.35 0.35 0.35>
    gravityOverride {z=-980}
    physicsBodyDefinitions [{
      boundBone {boneName=dyn_petals04_01_lfbk}
      boxExtents=<18 18 18>
      constraintSetup {angularLimitsMax=<0 40 20>, angularLimitsMin=<0 -40 -20>, angularTarget=<1 0 0>}
      localJointOffset=<3 0 0>
     }]
    relativeSpaceBone {boneName=base_flower_03_lfbk}
    simSpaceSettings {maxAngularAcceleration=35, maxAngularVelocity=10, simSpaceAngularAlpha=0.8, externalAngularVelocity {y=3}}
   }
   n37 = Anim Dynamics - Bone: dyn_petals05_01_lfbk!(
    ComponentPose=@n36
    Alpha=1
    GravityOverride:Vector="0,0,-980"
    ComponentLinearAccScale:Vector="0.35,0.35,0.35"
    ComponentLinearVelScale:Vector="0.35,0.35,0.35"
    ExternalForce="0,0,0"
   ) -> (Pose) {
    angularDampingOverride=0.9999
    angularSpringConstant=5000
    bAngularSpring=true
    bOverrideAngularDamping=true
    bUseGravityOverride=true
    gravityScale=0.35
    linearSpringConstant=10000
    simulationSpace=RootRelative
    worldSpaceMinimumScale=0
    boundBone {boneName=dyn_petals05_01_lfbk}
    componentAppliedLinearAccClamp=<35000 35000 35000>
    componentLinearAccScale=<0.35 0.35 0.35>
    componentLinearVelScale=<0.35 0.35 0.35>
    gravityOverride {z=-980}
    physicsBodyDefinitions [{
      boundBone {boneName=dyn_petals05_01_lfbk}
      boxExtents=<14 14 14>
      constraintSetup {angularLimitsMax=<0 20 0>, angularLimitsMin=<0 0 -20>, angularTarget=<1 0 0>}
      localJointOffset=<3 0 0>
     }]
    relativeSpaceBone {boneName=base_flower_03_lfbk}
    simSpaceSettings {maxAngularAcceleration=35, maxAngularVelocity=10, simSpaceAngularAlpha=0.8, externalAngularVelocity {y=3}}
   }
   n38 = Anim Dynamics - Bone: dyn_petals01_01_rt!(
    ComponentPose=@n37
    Alpha=1
    GravityOverride:Vector="0,0,-980"
    ComponentLinearAccScale:Vector="0.35,0.35,0.35"
    ComponentLinearVelScale:Vector="0.35,0.35,0.35"
    ExternalForce="0,0,0"
   ) -> (Pose) {
    angularDampingOverride=0.9999
    angularSpringConstant=5000
    bAngularSpring=true
    bOverrideAngularDamping=true
    bUseGravityOverride=true
    gravityScale=0.35
    linearSpringConstant=10000
    simulationSpace=RootRelative
    worldSpaceMinimumScale=0
    boundBone {boneName=dyn_petals01_01_rt}
    componentAppliedLinearAccClamp=<35000 35000 35000>
    componentLinearAccScale=<0.35 0.35 0.35>
    componentLinearVelScale=<0.35 0.35 0.35>
    gravityOverride {z=-980}
    physicsBodyDefinitions [{
      boundBone {boneName=dyn_petals01_01_rt}
      boxExtents=<14 14 14>
      constraintSetup {angularLimitsMax=<0 10 10>, angularLimitsMin=<0 -10 -10>, angularTarget=<1 0 0>}
      localJointOffset=<3 0 0>
     }]
    relativeSpaceBone {boneName=base_flower_04_rt}
    simSpaceSettings {maxAngularAcceleration=35, maxAngularVelocity=10, simSpaceAngularAlpha=0.8, externalAngularVelocity {y=3}}
   }
   n39 = Anim Dynamics - Bone: dyn_petals01_02_rt!(
    ComponentPose=@n38
    Alpha=1
    GravityOverride:Vector="0,0,-980"
    ComponentLinearAccScale:Vector="0.35,0.35,0.35"
    ComponentLinearVelScale:Vector="0.35,0.35,0.35"
    ExternalForce="0,0,0"
   ) -> (Pose) {
    angularDampingOverride=0.9999
    angularSpringConstant=5000
    bAngularSpring=true
    bOverrideAngularDamping=true
    bUseGravityOverride=true
    gravityScale=0.35
    linearSpringConstant=10000
    simulationSpace=RootRelative
    worldSpaceMinimumScale=0
    boundBone {boneName=dyn_petals01_02_rt}
    componentAppliedLinearAccClamp=<35000 35000 35000>
    componentLinearAccScale=<0.35 0.35 0.35>
    componentLinearVelScale=<0.35 0.35 0.35>
    gravityOverride {z=-980}
    physicsBodyDefinitions [{
      boundBone {boneName=dyn_petals01_02_rt}
      boxExtents=<12 12 12>
      constraintSetup {angularLimitsMax=<0 15 15>, angularLimitsMin=<0 -15 -15>, angularTarget=<1 0 0>}
      localJointOffset=<3 0 0>
     }]
    relativeSpaceBone {boneName=base_flower_04_rt}
    simSpaceSettings {maxAngularAcceleration=35, maxAngularVelocity=10, simSpaceAngularAlpha=0.8, externalAngularVelocity {y=3}}
   }
   n40 = Anim Dynamics - Bone: dyn_petals02_01_rt!(
    ComponentPose=@n39
    Alpha=1
    GravityOverride:Vector="0,0,-980"
    ComponentLinearAccScale:Vector="0.35,0.35,0.35"
    ComponentLinearVelScale:Vector="0.35,0.35,0.35"
    ExternalForce="0,0,0"
   ) -> (Pose) {
    angularDampingOverride=0.9999
    angularSpringConstant=5000
    bAngularSpring=true
    bOverrideAngularDamping=true
    bUseGravityOverride=true
    gravityScale=0.35
    linearSpringConstant=10000
    simulationSpace=RootRelative
    worldSpaceMinimumScale=0
    boundBone {boneName=dyn_petals02_01_rt}
    componentAppliedLinearAccClamp=<35000 35000 35000>
    componentLinearAccScale=<0.35 0.35 0.35>
    componentLinearVelScale=<0.35 0.35 0.35>
    gravityOverride {z=-980}
    physicsBodyDefinitions [{
      boundBone {boneName=dyn_petals02_01_rt}
      boxExtents=<14 14 14>
      constraintSetup {angularLimitsMax=<0 10 10>, angularLimitsMin=<0 -10 -10>, angularTarget=<1 0 0>}
      localJointOffset=<3 0 0>
     }]
    relativeSpaceBone {boneName=base_flower_04_rt}
    simSpaceSettings {maxAngularAcceleration=35, maxAngularVelocity=10, simSpaceAngularAlpha=0.8, externalAngularVelocity {y=3}}
   }
   n41 = Anim Dynamics - Bone: dyn_petals02_02_rt!(
    ComponentPose=@n40
    Alpha=1
    GravityOverride:Vector="0,0,-980"
    ComponentLinearAccScale:Vector="0.35,0.35,0.35"
    ComponentLinearVelScale:Vector="0.35,0.35,0.35"
    ExternalForce="0,0,0"
   ) -> (Pose) {
    angularDampingOverride=0.9999
    angularSpringConstant=5000
    bAngularSpring=true
    bOverrideAngularDamping=true
    bUseGravityOverride=true
    gravityScale=0.35
    linearSpringConstant=10000
    simulationSpace=RootRelative
    worldSpaceMinimumScale=0
    boundBone {boneName=dyn_petals02_02_rt}
    componentAppliedLinearAccClamp=<35000 35000 35000>
    componentLinearAccScale=<0.35 0.35 0.35>
    componentLinearVelScale=<0.35 0.35 0.35>
    gravityOverride {z=-980}
    physicsBodyDefinitions [{
      boundBone {boneName=dyn_petals02_02_rt}
      boxExtents=<12 12 12>
      constraintSetup {angularLimitsMax=<0 15 15>, angularLimitsMin=<0 -15 -15>, angularTarget=<1 0 0>}
      localJointOffset=<3 0 0>
     }]
    relativeSpaceBone {boneName=base_flower_04_rt}
    simSpaceSettings {maxAngularAcceleration=35, maxAngularVelocity=10, simSpaceAngularAlpha=0.8, externalAngularVelocity {y=3}}
   }
   n42 = Anim Dynamics - Bone: dyn_petals03_01_rt!(
    ComponentPose=@n41
    Alpha=1
    GravityOverride:Vector="0,0,-980"
    ComponentLinearAccScale:Vector="0.35,0.35,0.35"
    ComponentLinearVelScale:Vector="0.35,0.35,0.35"
    ExternalForce="0,0,0"
   ) -> (Pose) {
    angularDampingOverride=0.9999
    angularSpringConstant=5000
    bAngularSpring=true
    bOverrideAngularDamping=true
    bUseGravityOverride=true
    gravityScale=0.35
    linearSpringConstant=10000
    simulationSpace=RootRelative
    worldSpaceMinimumScale=0
    boundBone {boneName=dyn_petals03_01_rt}
    componentAppliedLinearAccClamp=<35000 35000 35000>
    componentLinearAccScale=<0.35 0.35 0.35>
    componentLinearVelScale=<0.35 0.35 0.35>
    gravityOverride {z=-980}
    physicsBodyDefinitions [{
      boundBone {boneName=dyn_petals03_01_rt}
      boxExtents=<16 16 16>
      constraintSetup {angularLimitsMax=<0 10 10>, angularLimitsMin=<0 -10 -10>, angularTarget=<1 0 0>}
      localJointOffset=<3 0 0>
     }]
    relativeSpaceBone {boneName=base_flower_04_rt}
    simSpaceSettings {maxAngularAcceleration=35, maxAngularVelocity=10, simSpaceAngularAlpha=0.8, externalAngularVelocity {y=3}}
   }
   n43 = Anim Dynamics - Bone: dyn_petals04_02_rt!(
    ComponentPose=@n42
    Alpha=1
    GravityOverride:Vector="0,0,-980"
    ComponentLinearAccScale:Vector="0.35,0.35,0.35"
    ComponentLinearVelScale:Vector="0.35,0.35,0.35"
    ExternalForce="0,0,0"
   ) -> (Pose) {
    angularDampingOverride=0.9999
    angularSpringConstant=5000
    bAngularSpring=true
    bOverrideAngularDamping=true
    bUseGravityOverride=true
    gravityScale=0.35
    linearSpringConstant=10000
    simulationSpace=RootRelative
    worldSpaceMinimumScale=0
    boundBone {boneName=dyn_petals04_02_rt}
    componentAppliedLinearAccClamp=<35000 35000 35000>
    componentLinearAccScale=<0.35 0.35 0.35>
    componentLinearVelScale=<0.35 0.35 0.35>
    gravityOverride {z=-980}
    physicsBodyDefinitions [{
      boundBone {boneName=dyn_petals04_02_rt}
      boxExtents=<16 16 16>
      constraintSetup {angularLimitsMax=<0 10 10>, angularLimitsMin=<0 -10 -10>, angularTarget=<1 0 0>}
      localJointOffset=<3 0 0>
     }]
    relativeSpaceBone {boneName=base_flower_04_rt}
    simSpaceSettings {maxAngularAcceleration=35, maxAngularVelocity=10, simSpaceAngularAlpha=0.8, externalAngularVelocity {y=3}}
   }
   n44 = Anim Dynamics - Bone: dyn_petals05_01_rt!(
    ComponentPose=@n43
    Alpha=1
    GravityOverride:Vector="0,0,-980"
    ComponentLinearAccScale:Vector="0.35,0.35,0.35"
    ComponentLinearVelScale:Vector="0.35,0.35,0.35"
    ExternalForce="0,0,0"
   ) -> (Pose) {
    angularDampingOverride=0.9999
    angularSpringConstant=5000
    bAngularSpring=true
    bOverrideAngularDamping=true
    bUseGravityOverride=true
    gravityScale=0.35
    linearSpringConstant=10000
    simulationSpace=RootRelative
    worldSpaceMinimumScale=0
    boundBone {boneName=dyn_petals05_01_rt}
    componentAppliedLinearAccClamp=<35000 35000 35000>
    componentLinearAccScale=<0.35 0.35 0.35>
    componentLinearVelScale=<0.35 0.35 0.35>
    gravityOverride {z=-980}
    physicsBodyDefinitions [{
      boundBone {boneName=dyn_petals05_01_rt}
      boxExtents=<16 16 16>
      constraintSetup {angularLimitsMax=<0 10 10>, angularLimitsMin=<0 -10 -10>, angularTarget=<1 0 0>}
      localJointOffset=<3 0 0>
     }]
    relativeSpaceBone {boneName=base_flower_04_rt}
    simSpaceSettings {maxAngularAcceleration=35, maxAngularVelocity=10, simSpaceAngularAlpha=0.8, externalAngularVelocity {y=3}}
   }
   n45 = Anim Dynamics - Bone: dyn_petals01_01_bk!(
    ComponentPose=@n44
    Alpha=1
    GravityOverride:Vector="0,0,-980"
    ComponentLinearAccScale:Vector="0.35,0.35,0.35"
    ComponentLinearVelScale:Vector="0.35,0.35,0.35"
    ExternalForce="0,0,0"
   ) -> (Pose) {
    angularDampingOverride=0.9999
    angularSpringConstant=5000
    bAngularSpring=true
    bOverrideAngularDamping=true
    bUseGravityOverride=true
    gravityScale=0.35
    linearSpringConstant=10000
    simulationSpace=RootRelative
    worldSpaceMinimumScale=0
    boundBone {boneName=dyn_petals01_01_bk}
    componentAppliedLinearAccClamp=<35000 35000 35000>
    componentLinearAccScale=<0.35 0.35 0.35>
    componentLinearVelScale=<0.35 0.35 0.35>
    gravityOverride {z=-980}
    physicsBodyDefinitions [{
      boundBone {boneName=dyn_petals01_01_bk}
      boxExtents=<16 16 16>
      constraintSetup {angularLimitsMax=<0 5 10>, angularLimitsMin=<0 -5 -10>, angularTarget=<1 0 0>}
      localJointOffset=<3 0 0>
     }]
    relativeSpaceBone {boneName=base_flower_05_bk}
    simSpaceSettings {maxAngularAcceleration=35, maxAngularVelocity=10, simSpaceAngularAlpha=0.8, externalAngularVelocity {y=3}}
   }
   n46 = Anim Dynamics - Bone: dyn_petals01_02_bk!(
    ComponentPose=@n45
    Alpha=1
    GravityOverride:Vector="0,0,-980"
    ComponentLinearAccScale:Vector="0.35,0.35,0.35"
    ComponentLinearVelScale:Vector="0.35,0.35,0.35"
    ExternalForce="0,0,0"
   ) -> (Pose) {
    angularDampingOverride=0.9999
    angularSpringConstant=5000
    bAngularSpring=true
    bOverrideAngularDamping=true
    bUseGravityOverride=true
    gravityScale=0.35
    linearSpringConstant=10000
    simulationSpace=RootRelative
    worldSpaceMinimumScale=0
    boundBone {boneName=dyn_petals01_02_bk}
    componentAppliedLinearAccClamp=<35000 35000 35000>
    componentLinearAccScale=<0.35 0.35 0.35>
    componentLinearVelScale=<0.35 0.35 0.35>
    gravityOverride {z=-980}
    physicsBodyDefinitions [{
      boundBone {boneName=dyn_petals01_02_bk}
      boxExtents=<14 14 14>
      constraintSetup {angularLimitsMax=<0 8 8>, angularLimitsMin=<0 -8 -8>, angularTarget=<1 0 0>}
      localJointOffset=<3 0 0>
     }]
    relativeSpaceBone {boneName=base_flower_05_bk}
    simSpaceSettings {maxAngularAcceleration=35, maxAngularVelocity=10, simSpaceAngularAlpha=0.8, externalAngularVelocity {y=3}}
   }
   n47 = Anim Dynamics - Bone: dyn_petals02_01_bk!(
    ComponentPose=@n46
    Alpha=1
    GravityOverride:Vector="0,0,-980"
    ComponentLinearAccScale:Vector="0.35,0.35,0.35"
    ComponentLinearVelScale:Vector="0.35,0.35,0.35"
    ExternalForce="0,0,0"
   ) -> (Pose) {
    angularDampingOverride=0.9999
    angularSpringConstant=5000
    bAngularSpring=true
    bOverrideAngularDamping=true
    bUseGravityOverride=true
    gravityScale=0.35
    linearSpringConstant=10000
    simulationSpace=RootRelative
    worldSpaceMinimumScale=0
    boundBone {boneName=dyn_petals02_01_bk}
    componentAppliedLinearAccClamp=<35000 35000 35000>
    componentLinearAccScale=<0.35 0.35 0.35>
    componentLinearVelScale=<0.35 0.35 0.35>
    gravityOverride {z=-980}
    physicsBodyDefinitions [{
      boundBone {boneName=dyn_petals02_01_bk}
      boxExtents=<20 20 20>
      constraintSetup {angularLimitsMax=<0 5 5>, angularLimitsMin=<0 -5 -5>, angularTarget=<1 0 0>}
      localJointOffset=<3 0 0>
     }]
    relativeSpaceBone {boneName=base_flower_05_bk}
    simSpaceSettings {maxAngularAcceleration=35, maxAngularVelocity=10, simSpaceAngularAlpha=0.8, externalAngularVelocity {y=3}}
   }
   n48 = Anim Dynamics - Bone: dyn_petals02_02_bk!(
    ComponentPose=@n47
    Alpha=1
    GravityOverride:Vector="0,0,-980"
    ComponentLinearAccScale:Vector="0.35,0.35,0.35"
    ComponentLinearVelScale:Vector="0.35,0.35,0.35"
    ExternalForce="0,0,0"
   ) -> (Pose) {
    angularDampingOverride=0.9999
    angularSpringConstant=5000
    bAngularSpring=true
    bOverrideAngularDamping=true
    bUseGravityOverride=true
    gravityScale=0.35
    linearSpringConstant=10000
    simulationSpace=RootRelative
    worldSpaceMinimumScale=0
    boundBone {boneName=dyn_petals02_02_bk}
    componentAppliedLinearAccClamp=<35000 35000 35000>
    componentLinearAccScale=<0.35 0.35 0.35>
    componentLinearVelScale=<0.35 0.35 0.35>
    gravityOverride {z=-980}
    physicsBodyDefinitions [{
      boundBone {boneName=dyn_petals02_02_bk}
      boxExtents=<18 18 18>
      constraintSetup {angularLimitsMax=<0 10 10>, angularLimitsMin=<0 -10 -10>, angularTarget=<1 0 0>}
      localJointOffset=<3 0 0>
     }]
    relativeSpaceBone {boneName=base_flower_05_bk}
    simSpaceSettings {maxAngularAcceleration=35, maxAngularVelocity=10, simSpaceAngularAlpha=0.8, externalAngularVelocity {y=3}}
   }
   n49 = Anim Dynamics - Bone: dyn_petals02_03_bk!(
    ComponentPose=@n48
    Alpha=1
    GravityOverride:Vector="0,0,-980"
    ComponentLinearAccScale:Vector="0.35,0.35,0.35"
    ComponentLinearVelScale:Vector="0.35,0.35,0.35"
    ExternalForce="0,0,0"
   ) -> (Pose) {
    angularDampingOverride=0.9999
    angularSpringConstant=5000
    bAngularSpring=true
    bOverrideAngularDamping=true
    bUseGravityOverride=true
    gravityScale=0.35
    linearSpringConstant=10000
    simulationSpace=RootRelative
    worldSpaceMinimumScale=0
    boundBone {boneName=dyn_petals02_03_bk}
    componentAppliedLinearAccClamp=<35000 35000 35000>
    componentLinearAccScale=<0.35 0.35 0.35>
    componentLinearVelScale=<0.35 0.35 0.35>
    gravityOverride {z=-980}
    physicsBodyDefinitions [{
      boundBone {boneName=dyn_petals02_03_bk}
      boxExtents=<16 16 16>
      constraintSetup {angularLimitsMax=<0 15 15>, angularLimitsMin=<0 -15 -15>, angularTarget=<1 0 0>}
      localJointOffset=<3 0 0>
     }]
    relativeSpaceBone {boneName=base_flower_05_bk}
    simSpaceSettings {maxAngularAcceleration=35, maxAngularVelocity=10, simSpaceAngularAlpha=0.8, externalAngularVelocity {y=3}}
   }
   n50 = Anim Dynamics - Bone: dyn_petals03_01_bk!(
    ComponentPose=@n49
    Alpha=1
    GravityOverride:Vector="0,0,-980"
    ComponentLinearAccScale:Vector="0.35,0.35,0.35"
    ComponentLinearVelScale:Vector="0.35,0.35,0.35"
    ExternalForce="0,0,0"
   ) -> (Pose) {
    angularDampingOverride=0.9999
    angularSpringConstant=5000
    bAngularSpring=true
    bOverrideAngularDamping=true
    bUseGravityOverride=true
    gravityScale=0.35
    linearSpringConstant=10000
    simulationSpace=RootRelative
    worldSpaceMinimumScale=0
    boundBone {boneName=dyn_petals03_01_bk}
    componentAppliedLinearAccClamp=<35000 35000 35000>
    componentLinearAccScale=<0.35 0.35 0.35>
    componentLinearVelScale=<0.35 0.35 0.35>
    gravityOverride {z=-980}
    physicsBodyDefinitions [{
      boundBone {boneName=dyn_petals03_01_bk}
      boxExtents=<14 14 14>
      constraintSetup {angularLimitsMax=<0 5 5>, angularLimitsMin=<0 -5 -5>, angularTarget=<1 0 0>}
      localJointOffset=<3 0 0>
     }]
    relativeSpaceBone {boneName=base_flower_05_bk}
    simSpaceSettings {maxAngularAcceleration=35, maxAngularVelocity=10, simSpaceAngularAlpha=0.8, externalAngularVelocity {y=3}}
   }
   n51 = Anim Dynamics - Bone: dyn_petals03_02_bk!(
    ComponentPose=@n50
    Alpha=1
    GravityOverride:Vector="0,0,-980"
    ComponentLinearAccScale:Vector="0.35,0.35,0.35"
    ComponentLinearVelScale:Vector="0.35,0.35,0.35"
    ExternalForce="0,0,0"
   ) -> (Pose) {
    angularDampingOverride=0.9999
    angularSpringConstant=5000
    bAngularSpring=true
    bOverrideAngularDamping=true
    bUseGravityOverride=true
    gravityScale=0.35
    linearSpringConstant=10000
    simulationSpace=RootRelative
    worldSpaceMinimumScale=0
    boundBone {boneName=dyn_petals03_02_bk}
    componentAppliedLinearAccClamp=<35000 35000 35000>
    componentLinearAccScale=<0.35 0.35 0.35>
    componentLinearVelScale=<0.35 0.35 0.35>
    gravityOverride {z=-980}
    physicsBodyDefinitions [{
      boundBone {boneName=dyn_petals03_02_bk}
      boxExtents=<12 12 12>
      constraintSetup {angularLimitsMax=<0 5 5>, angularLimitsMin=<0 -5 -5>, angularTarget=<1 0 0>}
      localJointOffset=<3 0 0>
     }]
    relativeSpaceBone {boneName=base_flower_05_bk}
    simSpaceSettings {maxAngularAcceleration=35, maxAngularVelocity=10, simSpaceAngularAlpha=0.8, externalAngularVelocity {y=3}}
   }
   n52 = Anim Dynamics - Bone: dyn_petals04_01_bk!(
    ComponentPose=@n51
    Alpha=1
    GravityOverride:Vector="0,0,-980"
    ComponentLinearAccScale:Vector="0.35,0.35,0.35"
    ComponentLinearVelScale:Vector="0.35,0.35,0.35"
    ExternalForce="0,0,0"
   ) -> (Pose) {
    angularDampingOverride=0.9999
    angularSpringConstant=5000
    bAngularSpring=true
    bOverrideAngularDamping=true
    bUseGravityOverride=true
    gravityScale=0.35
    linearSpringConstant=10000
    simulationSpace=RootRelative
    worldSpaceMinimumScale=0
    boundBone {boneName=dyn_petals04_01_bk}
    componentAppliedLinearAccClamp=<35000 35000 35000>
    componentLinearAccScale=<0.35 0.35 0.35>
    componentLinearVelScale=<0.35 0.35 0.35>
    gravityOverride {z=-980}
    physicsBodyDefinitions [{
      boundBone {boneName=dyn_petals04_01_bk}
      boxExtents=<14 14 14>
      constraintSetup {angularLimitsMax=<0 8 8>, angularLimitsMin=<0 -8 -8>, angularTarget=<1 0 0>}
      localJointOffset=<2 0 0>
     }]
    relativeSpaceBone {boneName=base_flower_05_bk}
    simSpaceSettings {maxAngularAcceleration=35, maxAngularVelocity=10, simSpaceAngularAlpha=0.8, externalAngularVelocity {y=3}}
   }
   n53 = Anim Dynamics - Bone: dyn_petals04_02_bk!(
    ComponentPose=@n52
    Alpha=1
    GravityOverride:Vector="0,0,-980"
    ComponentLinearAccScale:Vector="0.35,0.35,0.35"
    ComponentLinearVelScale:Vector="0.35,0.35,0.35"
    ExternalForce="0,0,0"
   ) -> (Pose) {
    angularDampingOverride=0.9999
    angularSpringConstant=5000
    bAngularSpring=true
    bOverrideAngularDamping=true
    bUseGravityOverride=true
    gravityScale=0.35
    linearSpringConstant=10000
    simulationSpace=RootRelative
    worldSpaceMinimumScale=0
    boundBone {boneName=dyn_petals04_02_bk}
    componentAppliedLinearAccClamp=<35000 35000 35000>
    componentLinearAccScale=<0.35 0.35 0.35>
    componentLinearVelScale=<0.35 0.35 0.35>
    gravityOverride {z=-980}
    physicsBodyDefinitions [{
      boundBone {boneName=dyn_petals04_02_bk}
      boxExtents=<12 12 12>
      constraintSetup {angularLimitsMax=<0 8 8>, angularLimitsMin=<0 -8 -8>, angularTarget=<1 0 0>}
      localJointOffset=<2 0 0>
     }]
    relativeSpaceBone {boneName=base_flower_05_bk}
    simSpaceSettings {maxAngularAcceleration=35, maxAngularVelocity=10, simSpaceAngularAlpha=0.8, externalAngularVelocity {y=3}}
   }
   n54 = Anim Dynamics - Bone: dyn_stamen01_01_cr!(
    ComponentPose=@n53
    Alpha=1
    GravityOverride:Vector="0,0,-980"
    ComponentLinearAccScale:Vector="0.35,0.35,0.35"
    ComponentLinearVelScale:Vector="0.35,0.35,0.35"
    ExternalForce="0,0,0"
   ) -> (Pose) {
    angularDampingOverride=0.9999
    angularSpringConstant=5000
    bOverrideAngularDamping=true
    bUseGravityOverride=true
    gravityScale=0.35
    linearSpringConstant=10000
    simulationSpace=RootRelative
    worldSpaceMinimumScale=0
    boundBone {boneName=dyn_stamen01_01_cr}
    componentAppliedLinearAccClamp=<35000 35000 35000>
    componentLinearAccScale=<0.35 0.35 0.35>
    componentLinearVelScale=<0.35 0.35 0.35>
    gravityOverride {z=-980}
    physicsBodyDefinitions [{
      boundBone {boneName=dyn_stamen01_01_cr}
      boxExtents=<18 18 18>
      constraintSetup {angularLimitsMax=<0 5 5>, angularLimitsMin=<0 -5 -5>, angularTarget=<1 0 0>}
      localJointOffset=<3 0 0>
     }]
    relativeSpaceBone {boneName=base_flower_05_bk}
    simSpaceSettings {maxAngularAcceleration=35, maxAngularVelocity=10, simSpaceAngularAlpha=0.8, externalAngularVelocity {y=3}}
   }
   n55 = Anim Dynamics - Bone: dyn_stamen01_02_cr!(
    ComponentPose=@n54
    Alpha=1
    GravityOverride:Vector="0,0,-980"
    ComponentLinearAccScale:Vector="0.35,0.35,0.35"
    ComponentLinearVelScale:Vector="0.35,0.35,0.35"
    ExternalForce="0,0,0"
   ) -> (Pose) {
    angularDampingOverride=0.9999
    angularSpringConstant=5000
    bAngularSpring=true
    bOverrideAngularDamping=true
    bUseGravityOverride=true
    gravityScale=0.35
    linearSpringConstant=10000
    simulationSpace=RootRelative
    worldSpaceMinimumScale=0
    boundBone {boneName=dyn_stamen01_02_cr}
    componentAppliedLinearAccClamp=<35000 35000 35000>
    componentLinearAccScale=<0.35 0.35 0.35>
    componentLinearVelScale=<0.35 0.35 0.35>
    gravityOverride {z=-980}
    physicsBodyDefinitions [{
      boundBone {boneName=dyn_stamen01_02_cr}
      boxExtents=<16 16 16>
      constraintSetup {angularLimitsMax=<0 20 20>, angularLimitsMin=<0 -20 -20>, angularTarget=<1 0 0>}
      localJointOffset=<3 0 0>
     }]
    relativeSpaceBone {boneName=base_flower_05_bk}
    simSpaceSettings {maxAngularAcceleration=35, maxAngularVelocity=10, simSpaceAngularAlpha=0.8, externalAngularVelocity {y=3}}
   }
   n56 = Anim Dynamics - Bone: dyn_stamen02_01_cr!(
    ComponentPose=@n55
    Alpha=1
    GravityOverride:Vector="0,0,-980"
    ComponentLinearAccScale:Vector="0.35,0.35,0.35"
    ComponentLinearVelScale:Vector="0.35,0.35,0.35"
    ExternalForce="0,0,0"
   ) -> (Pose) {
    angularDampingOverride=0.9999
    angularSpringConstant=5000
    bOverrideAngularDamping=true
    bUseGravityOverride=true
    gravityScale=0.35
    linearSpringConstant=10000
    simulationSpace=RootRelative
    worldSpaceMinimumScale=0
    boundBone {boneName=dyn_stamen02_01_cr}
    componentAppliedLinearAccClamp=<35000 35000 35000>
    componentLinearAccScale=<0.35 0.35 0.35>
    componentLinearVelScale=<0.35 0.35 0.35>
    gravityOverride {z=-980}
    physicsBodyDefinitions [{
      boundBone {boneName=dyn_stamen02_01_cr}
      boxExtents=<18 18 18>
      constraintSetup {angularLimitsMax=<0 5 5>, angularLimitsMin=<0 -5 -5>, angularTarget=<1 0 0>}
      localJointOffset=<2 0 0>
     }]
    relativeSpaceBone {boneName=base_flower_05_bk}
    simSpaceSettings {maxAngularAcceleration=35, maxAngularVelocity=10, simSpaceAngularAlpha=0.8, externalAngularVelocity {y=3}}
   }
   n57 = Anim Dynamics - Bone: dyn_stamen02_03_cr!(
    ComponentPose=@n56
    Alpha=1
    GravityOverride:Vector="0,0,-980"
    ComponentLinearAccScale:Vector="0.35,0.35,0.35"
    ComponentLinearVelScale:Vector="0.35,0.35,0.35"
    ExternalForce="0,0,0"
   ) -> (Pose) {
    angularDampingOverride=0.9999
    angularSpringConstant=5000
    bAngularSpring=true
    bOverrideAngularDamping=true
    bUseGravityOverride=true
    gravityScale=0.35
    linearSpringConstant=10000
    simulationSpace=RootRelative
    worldSpaceMinimumScale=0
    boundBone {boneName=dyn_stamen02_03_cr}
    componentAppliedLinearAccClamp=<35000 35000 35000>
    componentLinearAccScale=<0.35 0.35 0.35>
    componentLinearVelScale=<0.35 0.35 0.35>
    gravityOverride {z=-980}
    physicsBodyDefinitions [{
      boundBone {boneName=dyn_stamen02_03_cr}
      boxExtents=<16 16 16>
      constraintSetup {angularLimitsMax=<0 10 10>, angularLimitsMin=<0 -10 -10>, angularTarget=<1 0 0>}
      localJointOffset=<3 0 0>
     }]
    relativeSpaceBone {boneName=base_flower_05_bk}
    simSpaceSettings {maxAngularAcceleration=35, maxAngularVelocity=10, simSpaceAngularAlpha=0.8, externalAngularVelocity {y=3}}
   }
   n58 = Anim Dynamics - Bone: dyn_stamen03_01_cr!(
    ComponentPose=@n57
    Alpha=1
    GravityOverride:Vector="0,0,-980"
    ComponentLinearAccScale:Vector="0.35,0.35,0.35"
    ComponentLinearVelScale:Vector="0.35,0.35,0.35"
    ExternalForce="0,0,0"
   ) -> (Pose) {
    angularDampingOverride=0.9999
    angularSpringConstant=5000
    bOverrideAngularDamping=true
    bUseGravityOverride=true
    gravityScale=0.35
    linearSpringConstant=10000
    simulationSpace=RootRelative
    worldSpaceMinimumScale=0
    boundBone {boneName=dyn_stamen03_01_cr}
    componentAppliedLinearAccClamp=<35000 35000 35000>
    componentLinearAccScale=<0.35 0.35 0.35>
    componentLinearVelScale=<0.35 0.35 0.35>
    gravityOverride {z=-980}
    physicsBodyDefinitions [{
      boundBone {boneName=dyn_stamen03_01_cr}
      boxExtents=<18 18 18>
      constraintSetup {angularLimitsMax=<0 5 5>, angularLimitsMin=<0 -5 -5>, angularTarget=<1 0 0>}
      localJointOffset=<2 0 0>
     }]
    relativeSpaceBone {boneName=base_flower_05_bk}
    simSpaceSettings {maxAngularAcceleration=35, maxAngularVelocity=10, simSpaceAngularAlpha=0.8, externalAngularVelocity {y=3}}
   }
   n59 = Anim Dynamics - Bone: dyn_stamen03_03_cr!(
    ComponentPose=@n58
    Alpha=1
    GravityOverride:Vector="0,0,-980"
    ComponentLinearAccScale:Vector="0.35,0.35,0.35"
    ComponentLinearVelScale:Vector="0.35,0.35,0.35"
    ExternalForce="0,0,0"
   ) -> (Pose) {
    angularDampingOverride=0.9999
    angularSpringConstant=5000
    bAngularSpring=true
    bOverrideAngularDamping=true
    bUseGravityOverride=true
    gravityScale=0.35
    linearSpringConstant=10000
    simulationSpace=RootRelative
    worldSpaceMinimumScale=0
    boundBone {boneName=dyn_stamen03_03_cr}
    componentAppliedLinearAccClamp=<35000 35000 35000>
    componentLinearAccScale=<0.35 0.35 0.35>
    componentLinearVelScale=<0.35 0.35 0.35>
    gravityOverride {z=-980}
    physicsBodyDefinitions [{
      boundBone {boneName=dyn_stamen03_03_cr}
      boxExtents=<16 16 16>
      constraintSetup {angularLimitsMax=<0 20 20>, angularLimitsMin=<0 -20 -20>, angularTarget=<1 0 0>}
      localJointOffset=<3 0 0>
     }]
    relativeSpaceBone {boneName=base_flower_05_bk}
    simSpaceSettings {maxAngularAcceleration=35, maxAngularVelocity=10, simSpaceAngularAlpha=0.8, externalAngularVelocity {y=3}}
   }
   n60 = Anim Dynamics - Bone: dyn_stamen02_01_lfbk!(
    ComponentPose=@n59
    Alpha=1
    GravityOverride:Vector="0,0,-980"
    ComponentLinearAccScale:Vector="0.35,0.35,0.35"
    ComponentLinearVelScale:Vector="0.35,0.35,0.35"
    ExternalForce="0,0,0"
   ) -> (Pose) {
    angularDampingOverride=0.9999
    angularSpringConstant=5000
    bOverrideAngularDamping=true
    bUseGravityOverride=true
    gravityScale=0.35
    linearSpringConstant=10000
    simulationSpace=RootRelative
    worldSpaceMinimumScale=0
    boundBone {boneName=dyn_stamen02_01_lfbk}
    componentAppliedLinearAccClamp=<35000 35000 35000>
    componentLinearAccScale=<0.35 0.35 0.35>
    componentLinearVelScale=<0.35 0.35 0.35>
    gravityOverride {z=-980}
    physicsBodyDefinitions [{
      boundBone {boneName=dyn_stamen02_01_lfbk}
      boxExtents=<16 16 16>
      constraintSetup {angularLimitsMax=<0 5 5>, angularLimitsMin=<0 -5 -5>, angularTarget=<1 0 0>}
      localJointOffset=<3 0 0>
     }]
    relativeSpaceBone {boneName=base_flower_05_bk}
    simSpaceSettings {maxAngularAcceleration=35, maxAngularVelocity=10, simSpaceAngularAlpha=0.8, externalAngularVelocity {y=3}}
   }
   n61 = Anim Dynamics - Bone: dyn_stamen02_02_lfbk!(
    ComponentPose=@n60
    Alpha=1
    GravityOverride:Vector="0,0,-980"
    ComponentLinearAccScale:Vector="0.35,0.35,0.35"
    ComponentLinearVelScale:Vector="0.35,0.35,0.35"
    ExternalForce="0,0,0"
   ) -> (Pose) {
    angularDampingOverride=0.9999
    angularSpringConstant=5000
    bAngularSpring=true
    bOverrideAngularDamping=true
    bUseGravityOverride=true
    gravityScale=0.35
    linearSpringConstant=10000
    simulationSpace=RootRelative
    worldSpaceMinimumScale=0
    boundBone {boneName=dyn_stamen02_02_lfbk}
    componentAppliedLinearAccClamp=<35000 35000 35000>
    componentLinearAccScale=<0.35 0.35 0.35>
    componentLinearVelScale=<0.35 0.35 0.35>
    gravityOverride {z=-980}
    physicsBodyDefinitions [{
      boundBone {boneName=dyn_stamen02_02_lfbk}
      boxExtents=<14 14 14>
      constraintSetup {angularLimitsMax=<0 20 20>, angularLimitsMin=<0 -20 -20>, angularTarget=<1 0 0>}
      localJointOffset=<3 0 0>
     }]
    relativeSpaceBone {boneName=base_flower_05_bk}
    simSpaceSettings {maxAngularAcceleration=35, maxAngularVelocity=10, simSpaceAngularAlpha=0.8, externalAngularVelocity {y=3}}
   }
   n62 = Anim Dynamics - Bone: dyn_stamen01_01_lfbk!(
    ComponentPose=@n61
    Alpha=1
    GravityOverride:Vector="0,0,-980"
    ComponentLinearAccScale:Vector="0.35,0.35,0.35"
    ComponentLinearVelScale:Vector="0.35,0.35,0.35"
    ExternalForce="0,0,0"
   ) -> (Pose) {
    angularDampingOverride=0.9999
    angularSpringConstant=5000
    bOverrideAngularDamping=true
    bUseGravityOverride=true
    gravityScale=0.35
    linearSpringConstant=10000
    simulationSpace=RootRelative
    worldSpaceMinimumScale=0
    boundBone {boneName=dyn_stamen01_01_lfbk}
    componentAppliedLinearAccClamp=<35000 35000 35000>
    componentLinearAccScale=<0.35 0.35 0.35>
    componentLinearVelScale=<0.35 0.35 0.35>
    gravityOverride {z=-980}
    physicsBodyDefinitions [{
      boundBone {boneName=dyn_stamen01_01_lfbk}
      boxExtents=<16 16 16>
      constraintSetup {angularLimitsMax=<0 5 5>, angularLimitsMin=<0 -5 -5>, angularTarget=<1 0 0>}
      localJointOffset=<3 0 0>
     }]
    relativeSpaceBone {boneName=base_flower_05_bk}
    simSpaceSettings {maxAngularAcceleration=35, maxAngularVelocity=10, simSpaceAngularAlpha=0.8, externalAngularVelocity {y=3}}
   }
   n63 = Anim Dynamics - Bone: dyn_stamen01_02_lfbk!(
    ComponentPose=@n62
    Alpha=1
    GravityOverride:Vector="0,0,-980"
    ComponentLinearAccScale:Vector="0.35,0.35,0.35"
    ComponentLinearVelScale:Vector="0.35,0.35,0.35"
    ExternalForce="0,0,0"
   ) -> (Pose) {
    angularDampingOverride=0.9999
    angularSpringConstant=5000
    bAngularSpring=true
    bOverrideAngularDamping=true
    bUseGravityOverride=true
    gravityScale=0.35
    linearSpringConstant=10000
    simulationSpace=RootRelative
    worldSpaceMinimumScale=0
    boundBone {boneName=dyn_stamen01_02_lfbk}
    componentAppliedLinearAccClamp=<35000 35000 35000>
    componentLinearAccScale=<0.35 0.35 0.35>
    componentLinearVelScale=<0.35 0.35 0.35>
    gravityOverride {z=-980}
    physicsBodyDefinitions [{
      boundBone {boneName=dyn_stamen01_02_lfbk}
      boxExtents=<14 14 14>
      constraintSetup {angularLimitsMax=<0 20 20>, angularLimitsMin=<0 -20 -20>, angularTarget=<1 0 0>}
      localJointOffset=<3 0 0>
     }]
    relativeSpaceBone {boneName=base_flower_05_bk}
    simSpaceSettings {maxAngularAcceleration=35, maxAngularVelocity=10, simSpaceAngularAlpha=0.8, externalAngularVelocity {y=3}}
   }
   n64 = Anim Dynamics - Bone: dyn_stamen01_01_rt!(
    ComponentPose=@n63
    Alpha=1
    GravityOverride:Vector="0,0,-980"
    ComponentLinearAccScale:Vector="0.35,0.35,0.35"
    ComponentLinearVelScale:Vector="0.35,0.35,0.35"
    ExternalForce="0,0,0"
   ) -> (Pose) {
    angularDampingOverride=0.9999
    angularSpringConstant=2000
    bOverrideAngularDamping=true
    bUseGravityOverride=true
    gravityScale=0.35
    linearSpringConstant=10000
    simulationSpace=RootRelative
    worldSpaceMinimumScale=0
    boundBone {boneName=dyn_stamen01_01_rt}
    componentAppliedLinearAccClamp=<35000 35000 35000>
    componentLinearAccScale=<0.35 0.35 0.35>
    componentLinearVelScale=<0.35 0.35 0.35>
    gravityOverride {z=-980}
    physicsBodyDefinitions [{
      boundBone {boneName=dyn_stamen01_01_rt}
      boxExtents=<16 16 16>
      constraintSetup {angularLimitsMax=<0 5 5>, angularLimitsMin=<0 -5 -5>, angularTarget=<1 0 0>}
      localJointOffset=<3 0 0>
     }]
    relativeSpaceBone {boneName=base_flower_05_bk}
    simSpaceSettings {maxAngularAcceleration=35, maxAngularVelocity=10, simSpaceAngularAlpha=0.8, externalAngularVelocity {y=3}}
   }
   n65 = Anim Dynamics - Bone: dyn_stamen01_02_rt!(
    ComponentPose=@n64
    Alpha=1
    GravityOverride:Vector="0,0,-980"
    ComponentLinearAccScale:Vector="0.35,0.35,0.35"
    ComponentLinearVelScale:Vector="0.35,0.35,0.35"
    ExternalForce="0,0,0"
   ) -> (Pose) {
    angularDampingOverride=0.9999
    angularSpringConstant=5000
    bAngularSpring=true
    bOverrideAngularDamping=true
    bUseGravityOverride=true
    gravityScale=0.35
    linearSpringConstant=10000
    simulationSpace=RootRelative
    worldSpaceMinimumScale=0
    boundBone {boneName=dyn_stamen01_02_rt}
    componentAppliedLinearAccClamp=<35000 35000 35000>
    componentLinearAccScale=<0.35 0.35 0.35>
    componentLinearVelScale=<0.35 0.35 0.35>
    gravityOverride {z=-980}
    physicsBodyDefinitions [{
      boundBone {boneName=dyn_stamen01_02_rt}
      boxExtents=<14 14 14>
      constraintSetup {angularLimitsMax=<0 20 20>, angularLimitsMin=<0 -20 -20>, angularTarget=<1 0 0>}
      localJointOffset=<3 0 0>
     }]
    relativeSpaceBone {boneName=base_flower_05_bk}
    simSpaceSettings {maxAngularAcceleration=35, maxAngularVelocity=10, simSpaceAngularAlpha=0.8, externalAngularVelocity {y=3}}
   }
   n66 = Anim Dynamics - Bone: dyn_stamen02_01_rt!(
    ComponentPose=@n65
    Alpha=1
    GravityOverride:Vector="0,0,-980"
    ComponentLinearAccScale:Vector="0.35,0.35,0.35"
    ComponentLinearVelScale:Vector="0.35,0.35,0.35"
    ExternalForce="0,0,0"
   ) -> (Pose) {
    angularDampingOverride=0.9999
    angularSpringConstant=2000
    bOverrideAngularDamping=true
    bUseGravityOverride=true
    gravityScale=0.35
    linearSpringConstant=10000
    simulationSpace=RootRelative
    worldSpaceMinimumScale=0
    boundBone {boneName=dyn_stamen02_01_rt}
    componentAppliedLinearAccClamp=<35000 35000 35000>
    componentLinearAccScale=<0.35 0.35 0.35>
    componentLinearVelScale=<0.35 0.35 0.35>
    gravityOverride {z=-980}
    physicsBodyDefinitions [{
      boundBone {boneName=dyn_stamen02_01_rt}
      boxExtents=<16 16 16>
      constraintSetup {angularLimitsMax=<0 5 5>, angularLimitsMin=<0 -5 -5>, angularTarget=<1 0 0>}
      localJointOffset=<2 0 0>
     }]
    relativeSpaceBone {boneName=base_flower_05_bk}
    simSpaceSettings {maxAngularAcceleration=35, maxAngularVelocity=10, simSpaceAngularAlpha=0.8, externalAngularVelocity {y=3}}
   }
   n67 = Anim Dynamics - Bone: dyn_stamen02_02_rt!(
    ComponentPose=@n66
    Alpha=1
    GravityOverride:Vector="0,0,-980"
    ComponentLinearAccScale:Vector="0.35,0.35,0.35"
    ComponentLinearVelScale:Vector="0.35,0.35,0.35"
    ExternalForce="0,0,0"
   ) -> (Pose) {
    angularDampingOverride=0.9999
    angularSpringConstant=5000
    bAngularSpring=true
    bOverrideAngularDamping=true
    bUseGravityOverride=true
    gravityScale=0.35
    linearSpringConstant=10000
    simulationSpace=RootRelative
    worldSpaceMinimumScale=0
    boundBone {boneName=dyn_stamen02_02_rt}
    componentAppliedLinearAccClamp=<35000 35000 35000>
    componentLinearAccScale=<0.35 0.35 0.35>
    componentLinearVelScale=<0.35 0.35 0.35>
    gravityOverride {z=-980}
    physicsBodyDefinitions [{
      boundBone {boneName=dyn_stamen02_02_rt}
      boxExtents=<14 14 14>
      constraintSetup {angularLimitsMax=<0 20 20>, angularLimitsMin=<0 -20 -20>, angularTarget=<1 0 0>}
      localJointOffset=<2 0 0>
     }]
    relativeSpaceBone {boneName=base_flower_05_bk}
    simSpaceSettings {maxAngularAcceleration=35, maxAngularVelocity=10, simSpaceAngularAlpha=0.8, externalAngularVelocity {y=3}}
   }
   n68 = Anim Dynamics - Bone: dyn_stamen01_01_bk!(
    ComponentPose=@n67
    Alpha=1
    GravityOverride:Vector="0,0,-980"
    ComponentLinearAccScale:Vector="0.35,0.35,0.35"
    ComponentLinearVelScale:Vector="0.35,0.35,0.35"
    ExternalForce="0,0,0"
   ) -> (Pose) {
    angularDampingOverride=0.9999
    angularSpringConstant=5000
    bAngularSpring=true
    bOverrideAngularDamping=true
    bUseGravityOverride=true
    gravityScale=0.35
    linearSpringConstant=10000
    simulationSpace=RootRelative
    worldSpaceMinimumScale=0
    boundBone {boneName=dyn_stamen01_01_bk}
    componentAppliedLinearAccClamp=<35000 35000 35000>
    componentLinearAccScale=<0.35 0.35 0.35>
    componentLinearVelScale=<0.35 0.35 0.35>
    gravityOverride {z=-980}
    physicsBodyDefinitions [{
      boundBone {boneName=dyn_stamen01_01_bk}
      boxExtents=<18 18 18>
      constraintSetup {angularLimitsMax=<0 0 15>, angularLimitsMin=<0 -15 -15>, angularTarget=<1 0 0>}
      localJointOffset=<2.5 0 0>
     }]
    relativeSpaceBone {boneName=base_flower_05_bk}
    simSpaceSettings {maxAngularAcceleration=35, maxAngularVelocity=10, simSpaceAngularAlpha=0.8, externalAngularVelocity {y=3}}
   }
   n69 = Anim Dynamics - Bone: dyn_stamen01_02_bk!(
    ComponentPose=@n68
    Alpha=1
    GravityOverride:Vector="0,0,-980"
    ComponentLinearAccScale:Vector="0.35,0.35,0.35"
    ComponentLinearVelScale:Vector="0.35,0.35,0.35"
    ExternalForce="0,0,0"
   ) -> (Pose) {
    angularDampingOverride=0.9999
    angularSpringConstant=5000
    bAngularSpring=true
    bOverrideAngularDamping=true
    bUseGravityOverride=true
    gravityScale=0.35
    linearSpringConstant=10000
    simulationSpace=RootRelative
    worldSpaceMinimumScale=0
    boundBone {boneName=dyn_stamen01_02_bk}
    componentAppliedLinearAccClamp=<35000 35000 35000>
    componentLinearAccScale=<0.35 0.35 0.35>
    componentLinearVelScale=<0.35 0.35 0.35>
    gravityOverride {z=-980}
    physicsBodyDefinitions [{
      boundBone {boneName=dyn_stamen01_02_bk}
      boxExtents=<16 16 16>
      constraintSetup {angularLimitsMax=<0 8 20>, angularLimitsMin=<0 -20 -20>, angularTarget=<1 0 0>}
      localJointOffset=<2.5 0 0>
     }]
    relativeSpaceBone {boneName=base_flower_05_bk}
    simSpaceSettings {maxAngularAcceleration=35, maxAngularVelocity=10, simSpaceAngularAlpha=0.8, externalAngularVelocity {y=3}}
   }
   n70 = Anim Dynamics - Bone: dyn_stamen02_01_bk!(
    ComponentPose=@n69
    Alpha=1
    GravityOverride:Vector="0,0,-980"
    ComponentLinearAccScale:Vector="0.35,0.35,0.35"
    ComponentLinearVelScale:Vector="0.35,0.35,0.35"
    ExternalForce="0,0,0"
   ) -> (Pose) {
    angularDampingOverride=0.9999
    angularSpringConstant=5000
    bAngularSpring=true
    bOverrideAngularDamping=true
    bUseGravityOverride=true
    gravityScale=0.35
    linearSpringConstant=10000
    simulationSpace=RootRelative
    worldSpaceMinimumScale=0
    boundBone {boneName=dyn_stamen02_01_bk}
    componentAppliedLinearAccClamp=<35000 35000 35000>
    componentLinearAccScale=<0.35 0.35 0.35>
    componentLinearVelScale=<0.35 0.35 0.35>
    gravityOverride {z=-980}
    physicsBodyDefinitions [{
      boundBone {boneName=dyn_stamen02_01_bk}
      boxExtents=<18 18 18>
      constraintSetup {angularLimitsMax=<0 0 8>, angularLimitsMin=<0 -8 -8>, angularTarget=<1 0 0>}
      localJointOffset=<2.5 0 0>
     }]
    relativeSpaceBone {boneName=base_flower_05_bk}
    simSpaceSettings {maxAngularAcceleration=35, maxAngularVelocity=10, simSpaceAngularAlpha=0.8, externalAngularVelocity {y=3}}
   }
   n71 = Anim Dynamics - Bone: dyn_stamen02_02_bk!(
    ComponentPose=@n70
    Alpha=1
    GravityOverride:Vector="0,0,-980"
    ComponentLinearAccScale:Vector="0.35,0.35,0.35"
    ComponentLinearVelScale:Vector="0.35,0.35,0.35"
    ExternalForce="0,0,0"
   ) -> (Pose) {
    angularDampingOverride=0.9999
    angularSpringConstant=5000
    bAngularSpring=true
    bOverrideAngularDamping=true
    bUseGravityOverride=true
    gravityScale=0.35
    linearSpringConstant=10000
    simulationSpace=RootRelative
    worldSpaceMinimumScale=0
    boundBone {boneName=dyn_stamen02_02_bk}
    componentAppliedLinearAccClamp=<35000 35000 35000>
    componentLinearAccScale=<0.35 0.35 0.35>
    componentLinearVelScale=<0.35 0.35 0.35>
    gravityOverride {z=-980}
    physicsBodyDefinitions [{
      boundBone {boneName=dyn_stamen02_02_bk}
      boxExtents=<16 16 16>
      constraintSetup {angularLimitsMax=<0 8 20>, angularLimitsMin=<0 -20 -20>, angularTarget=<1 0 0>}
      localJointOffset=<2.5 0 0>
     }]
    relativeSpaceBone {boneName=base_flower_05_bk}
    simSpaceSettings {maxAngularAcceleration=35, maxAngularVelocity=10, simSpaceAngularAlpha=0.8, externalAngularVelocity {y=3}}
   }
   n72 = Anim Dynamics - Bone: dyn_stamen03_01_bk!(
    ComponentPose=@n71
    Alpha=1
    GravityOverride:Vector="0,0,-980"
    ComponentLinearAccScale:Vector="0.35,0.35,0.35"
    ComponentLinearVelScale:Vector="0.35,0.35,0.35"
    ExternalForce="0,0,0"
   ) -> (Pose) {
    angularDampingOverride=0.9999
    angularSpringConstant=5000
    bAngularSpring=true
    bOverrideAngularDamping=true
    bUseGravityOverride=true
    gravityScale=0.35
    linearSpringConstant=10000
    simulationSpace=RootRelative
    worldSpaceMinimumScale=0
    boundBone {boneName=dyn_stamen03_01_bk}
    componentAppliedLinearAccClamp=<35000 35000 35000>
    componentLinearAccScale=<0.35 0.35 0.35>
    componentLinearVelScale=<0.35 0.35 0.35>
    gravityOverride {z=-980}
    physicsBodyDefinitions [{
      boundBone {boneName=dyn_stamen03_01_bk}
      boxExtents=<18 18 18>
      constraintSetup {angularLimitsMax=<0 0 15>, angularLimitsMin=<0 -15 -15>, angularTarget=<1 0 0>}
      localJointOffset=<2.5 0 0>
     }]
    relativeSpaceBone {boneName=base_flower_05_bk}
    simSpaceSettings {maxAngularAcceleration=35, maxAngularVelocity=10, simSpaceAngularAlpha=0.8, externalAngularVelocity {y=3}}
   }
   n73 = Anim Dynamics - Bone: dyn_stamen03_02_bk!(
    ComponentPose=@n72
    Alpha=1
    GravityOverride:Vector="0,0,-980"
    ComponentLinearAccScale:Vector="0.35,0.35,0.35"
    ComponentLinearVelScale:Vector="0.35,0.35,0.35"
    ExternalForce="0,0,0"
   ) -> (Pose) {
    angularDampingOverride=0.9999
    angularSpringConstant=5000
    bAngularSpring=true
    bOverrideAngularDamping=true
    bUseGravityOverride=true
    gravityScale=0.35
    linearSpringConstant=10000
    simulationSpace=RootRelative
    worldSpaceMinimumScale=0
    boundBone {boneName=dyn_stamen03_02_bk}
    componentAppliedLinearAccClamp=<35000 35000 35000>
    componentLinearAccScale=<0.35 0.35 0.35>
    componentLinearVelScale=<0.35 0.35 0.35>
    gravityOverride {z=-980}
    physicsBodyDefinitions [{
      boundBone {boneName=dyn_stamen03_02_bk}
      boxExtents=<16 16 16>
      constraintSetup {angularLimitsMax=<0 8 20>, angularLimitsMin=<0 -20 -20>, angularTarget=<1 0 0>}
      localJointOffset=<2.5 0 0>
     }]
    relativeSpaceBone {boneName=base_flower_05_bk}
    simSpaceSettings {maxAngularAcceleration=35, maxAngularVelocity=10, simSpaceAngularAlpha=0.8, externalAngularVelocity {y=3}}
   }
   Output Pose(Result=Component To Local(ComponentPose=@n73)) {layerGroup=None}
  ]}
 EventGraph {nodes [
   Event BlueprintUpdateAnimation()
   > n1 = Sequence()
   > n2 = PlayerMovement(AnimInstance=Self-Reference(), XDotProductCoefficient=100, YDotProductCoefficient=100, ZDotProductCoefficient=100)
   n3 = NOT Boolean(A=OR Boolean!(
     A=Get bIsParachuteOpened()
     B=Get bIsSkydiveDiveMode()
     C:bool=Get bIsSkydiveFloating()
     D:bool=Get bIsFalling()
     E:bool=Get bIsJumping()
     F:bool=Get OnZipline() -> (OnZipline:bool)
    ) -> (ReturnValue))
   n4 = BreakVector(InVec=MakeVector(X=@n2, Y=@n2.YMovement, Z=@n2.ZMovement))
   > Set GoingUp(GoingUp:bool=AND Boolean(A="float > float"(A=@n4.Z, B=10), B=@n3)) -> (Output_Get:bool)
   > Set GoingDown(GoingDown:bool=AND Boolean(A="float < float"(A=@n4.Z, B=-10), B=@n3)) -> (Output_Get:bool)
   Set Up Back Dist(execute=@n1.then_1, Up Back Dist:double="To Float (Boolean)"(InBool=Get GoingUp() -> (GoingUp:bool))) -> (Output_Get:double)
   > n5 = Branch(Condition=OR Boolean!(A=Get GoingUp(), B=Get GoingDown() -> (GoingDown:bool)) -> (ReturnValue))
   > Set animDynAlpha(animDynAlpha:double=0) -> (Output_Get:double)
   Set animDynAlpha(execute=@n5.else, animDynAlpha=1)
   n6 = Event BlueprintProcessAnimInputEvent()
   n7 = IsNamed(self=@n6.AnimInputEvent, InEventName=aie.Player.Zipline.Entered)
   n8 = OR Boolean(A=Get bIsMoving2D(), B=Get bIsTacticalSprinting())
   n9 = Get bIsJumping()
   > n10 = Branch(Condition=OR Boolean(A=@n7, B=AND Boolean!(A=@n7, B=@n9, C:bool=@n8) -> (ReturnValue)))
   > Set animDyn on zipline_3_6(animDyn on zipline_3_6:Vector="0, 0, 0") -> (Output_Get:Vector)
   > Set animDyn on zipline_3_0(animDyn on zipline_3_0:Vector="0, 0, 0") -> (Output_Get:Vector)
   > Set OnZipline(OnZipline:bool=true) -> (Output_Get:bool)
   Set animDyn on zipline_3_6(execute=[@n10.else, @n11.then], animDyn on zipline_3_6="0.25,0.25,0.25")
   > Set animDyn on zipline_3_0(animDyn on zipline_3_0="0.2,0.2,0.2")
   > Set OnZipline(OnZipline=false)
   Event OnEndSkydiving()
   > Set backDist(backDist:double=0) -> (Output_Get:double)
   Event OnBeginSkydiving()
   > Set backDist(backDist=1)
   n11 = Branch(Condition=IsNamed(self=@n6.AnimInputEvent, InEventName=aie.Player.Zipline.Exited))
   Reroute Node(InputPin:Wildcard) -> (OutputPin:Wildcard)
  ]}
]
variables [
 animDyn on zipline_3_0 {type=Vector, default=<0.2 0.2 0.2>}
 animDyn on zipline_3_6 {type=Vector, default=<0.25 0.25 0.25>}
 animDynAlpha {type=number, default=1}
 backDist {type=number}
 GoingDown {type=boolean}
 GoingUp {type=boolean}
 Gravity {type=Vector, default=<0 0 -160>}
 OnZipline {type=boolean}
 Up Back Dist {type=number}
]
